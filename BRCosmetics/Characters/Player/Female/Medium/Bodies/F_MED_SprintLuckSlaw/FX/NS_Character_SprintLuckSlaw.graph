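# Niagara system for the SprintLuckSlaw character body cosmetic, adding audio-reactive blue glows, drifting stars and mesh swirls around the body.
# All three emitters place particles at skeletal sockets through a shared socket reader, using per-group index and offset arrays.
# A system-level AudioReactivity value, with a fake audio loop module available, drives spawning, size and opacity.
# User flags zero selected groups for hidden hands or feet; time-of-day and front-end modifiers adjust brightness and motion.
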NiagaraSystem /BRCosmetics/Characters/Player/Female/Medium/Bodies/F_MED_SprintLuckSlaw/FX/NS_Character_SprintLuckSlaw
dependencies {
 usedDataInterfaces [
  &NiagaraDataInterfaceArrayFloat
  &NiagaraDataInterfaceCurve
  &NiagaraDataInterfaceArrayFloat3
  &NiagaraDataInterfaceArrayInt32
  &NiagaraDataInterfaceSocketReader
  &NiagaraDataInterfaceColorCurve
 ]
 usedDynamicInputs [
  &RandomRangeLinearColor
  &MakeVector2D
  /Niagara/DynamicInputs/Lerp/SmoothLerpOverTime_Float
  &ColorFromCurve
  /Game/Effects/Niagara/DynamicInputs/Math/FN_InterpFloat
  &VectorLength
  &NormalizeVector
 ]
 usedModules [
  &self:SystemSpawnScript.SystemScriptSource.SystemScriptGraph.NiagaraNodeAssignment_4.SetVariables_#1
  &self:SystemSpawnScript.SystemScriptSource.SystemScriptGraph.NiagaraNodeAssignment_0.SetVariables_#2
  &self:SystemSpawnScript.SystemScriptSource.SystemScriptGraph.NiagaraNodeAssignment_2.SetVariables_#3
  &SystemState
  &Timeline
  /Game/Effects/Niagara/Modules/Audio/FN_FakeAudioReactivityLoop
  &self:SystemSpawnScript.SystemScriptSource.SystemScriptGraph.NiagaraNodeAssignment_1.SetVariables_#4
  &self:SystemSpawnScript.SystemScriptSource.SystemScriptGraph.NiagaraNodeAssignment_10.SetVariables_#5
  &self:SystemSpawnScript.SystemScriptSource.SystemScriptGraph.NiagaraNodeAssignment_12.SetVariables_#6
  &self:SystemSpawnScript.SystemScriptSource.SystemScriptGraph.NiagaraNodeAssignment_5.SetVariables_#7
  &self:SystemSpawnScript.SystemScriptSource.SystemScriptGraph.NiagaraNodeAssignment_6.SetVariables_#8
 ]
}
emitters {
 Glow {
  emitterUpdateScript [
   &EmitterState(Inactive Response=::NewEnumerator1, Life Cycle Mode=::NewEnumerator1, Loop Duration=0.25, UseLoopDelay=false)
   &SpawnBurst_Instantaneous(
    Spawn Count=14
    Spawn Probability=&MakeCustomFloatFromBool(Bool=&SetBoolByFloatComparison(A=$System.AudioReactivity, B=0.1), True Float=1, False Float=0)
    Use Spawn Probability=true
    UseLoopCountLimit=false
   )
  ]
  particleSpawnScript [
   &PartitionParticles(Number Of Partitions=14)
   &InitializeParticle(
    Color=<rgba 0.2006 0.5642 0.8 0.5>
    Lifetime=0.25
    Sprite Size Mode=::NewEnumerator3
    Uniform Sprite Size=&SelectFloatFromArray(
     Direct Array Index=$Particles.PartitionParticles.Partition
     Float Selection Array=DataInterface() {propertyValues {FloatData [24,28,23,24,28,23,39,39,17,17,17,17,17,17,35,35]}}
    )
    UsePositionOffset=false
    Write Material Random=true
   )
   &CameraOffset(Camera Offset Amount=-2)
  ]
  particleUpdateScript [
   &ParticleState()
   &ScaleSpriteSize(
    Scale Sprite Size Mode=::NewEnumerator0
    Uniform Scale Factor=&MakeCustomFloatFromBool(
     Bool=$User.bisHandHidden
     True Float=&SelectFloatFromArray(
      Direct Array Index=&SpawnGroup()
      Float Selection Array=DataInterface() {propertyValues {FloatData [1,1,1,1,1,1,1,1,0,0,0,0,0,0]}}
     )
     False Float=1
    )
   ) {module=ScaleSpriteSize001}
   &ScaleSpriteSize(
    Scale Sprite Size Mode=::NewEnumerator0
    Uniform Scale Factor=&MakeCustomFloatFromBool(
     Bool=$User.bIsFootHidden
     True Float=&SelectFloatFromArray(
      Direct Array Index=&SpawnGroup()
      Float Selection Array=DataInterface() {propertyValues {FloatData [1,1,1,1,1,1,1,1,1,1,1,1,1,1]}}
     )
     False Float=1
    )
   ) {module=ScaleSpriteSize002}
   &SocketLocation(
    SamplePosOffset=true
    Sampled Position Offset=&SelectVectorFromArray(
     Direct Array Index=$Particles.PartitionParticles.Partition
     Vector Selection Array=DataInterface() {propertyValues {FloatData [
        <-28 0 0>
        <-15.8 0 0>
        <-3.6 0 0>
        <28 0 0>
        <15.8 0 0>
        <3.6 0 0>
        <0 -1 4.1037>
        <0 -1 -4>
        <-6.1441 0 0>
        <-12.5294 0 0>
        <-17.9959 0 0>
        <6.1441 0 0>
        <12.5294 0 0>
        <17.9959 0 0>
        <7.2287 5.4667 8.1216>
        <24.2993 8.7708 -0.6667>
       ]}}
    )
    Socket Index=&SelectIntFromArray(
     Direct Array Index=$Particles.PartitionParticles.Partition
     Int Selection Array=DataInterface() {propertyValues {IntData [0,0,0,5,5,5,7,7,3,3,3,8,8,8,7,9]}}
    )
    Socket Reader=$System.Socket Reader
    Socket Sampling Mode=::NewEnumerator3
    VelocitySpeedLimit=false
   )
   &ScaleColor(
    Scale Alpha=&Multiply_Float(
     A=&Multiply_Float(
      A=&Lerp_Float(A=0, B=1, Alpha=$System.AudioReactivity)
      B=&FNTimeOfDayScalar(Night Scale=3, Morning Scale=1, Day Scale=0.5, Evening Scale=1)
     )
     B=&FN_FrontEndFloatModifier(Gameplay=1, Front End=0.5)
    )
    ScaleA=true
    ScaleRGB=true
   )
   &ScaleSpriteSize(
    Scale Sprite Size Mode=::NewEnumerator0
    Uniform Scale Factor=&Multiply_Float(
     A=&MakeCustomFloatFromBool(
      Bool=$User.bisHandHidden
      True Float=&SelectFloatFromArray(
       Direct Array Index=$Particles.PartitionParticles.Partition
       Float Selection Array=DataInterface() {propertyValues {FloatData [0,0,1,1,0,0,1,1,1,1,1,1,1,1,1,1]}}
      )
      False Float=1
     )
     B=1
    )
   )
  ]
  rendererClasses [&NiagaraSpriteRendererProperties]
  renderers [{rendererClass=&NiagaraSpriteRendererProperties}]
 }
 Stars {
  emitterUpdateScript [
   &EmitterState(
    Inactive Response=::NewEnumerator1
    Life Cycle Mode=::NewEnumerator1
    Loop Behavior=::NewEnumerator1
    Loop Duration Mode=::NewEnumerator1
    UseLoopDelay=false
   )
   &SpawnRate(
    Spawn Probability=&Lerp_Float(A=0, B=0.7, Alpha=$System.AudioReactivity)
    SpawnRate=20
    Use Spawn Probability=true
   )
  ]
  particleSpawnScript [
   &FN_SetSpawnGroup(Number of Spawn Groups=8)
   /SetVariables(Particles.RandomInt:NiagaraInt32=&RandomRangeInt(Minimum=0, Maximum=7)) {isSetParameters=true}
   &InitializeParticle(
    AdjustAlpha=false
    AdjustHue=true
    AdjustSaturation=true
    AdjustValue=true
    Color=<rgba 0.5702 2.181 3 1>
    Color Mode=::NewEnumerator3
    Hue Shift Range:Vector2f=<-0.05 0.05>
    Lifetime Max=1.2
    Lifetime Mode=::NewEnumerator1
    Sprite Size Mode=::NewEnumerator3
    Sprite UV Mode=::NewEnumerator3
    Uniform Sprite Size=&Multiply_Float(
     A=&RandomRangeFloat(Minimum=15, Maximum=20)
     B=&SelectFloatFromArray(Direct Array Index=$Engine.QualityLevel, Float Selection Array=DataInterface() {propertyValues {FloatData [0.5,0.6,0.8,1,1]}})
    )
    UsePositionOffset=false
    Write Material Random=true
   )
   &CameraOffset(Camera Offset Amount=5)
   &SocketLocation(
    SamplePosOffset=true
    Sampled Position Offset=&MergeHelper_VectorSwitch(
     0=&RandomRangeVector(Minimum=<-35 -2 -2>, Maximum=<0 2 2>)
     1=&RandomRangeVector(Minimum=<0 -2 -2>, Maximum=<35 2 2>)
     2=&RandomRangeVector(Minimum=<-40 -3 -3>, Maximum=<0 3 3>)
     3=&RandomRangeVector(Minimum=<0 -3 -3>, Maximum=<40 3 3>)
     4=&RandomRangeVector(Minimum=<-20 -2 -2>, Maximum=<0 2 2>)
     5=&RandomRangeVector(Minimum=<0 -2 -2>, Maximum=<20 2 2>)
     6=&RandomRangeVector(Minimum=<-20 -2 -2>, Maximum=<0 2 2>)
     7=&RandomRangeVector(Minimum=<0 -2 -2>, Maximum=<20 2 2>)
     Index=&SpawnGroup()
     Num=::NewEnumerator6
    )
    Socket Index=&MergeHelper_IntSwitch(0=0, 1=5, 2=1, 3=6, 4=3, 5=8, 6=4, 7=9, Index=&SpawnGroup(), Num=::NewEnumerator6)
    Socket Reader=$System.Socket Reader
    Socket Sampling Mode=::NewEnumerator3
    VelocitySpeedLimit=false
   )
   &AddVelocity(
    Velocity=&FN_FrontEndVector3Modifier(Gameplay=<0 -20 10>, Front End=<0 0 30>)
    Velocity Speed Scale=1.4
   )
  ]
  particleUpdateScript [
   &ParticleState()
   &ScaleSpriteSize(
    Uniform Curve Scale=&Lerp_Float(A=0.3, B=1, Alpha=$System.AudioReactivity)
    Uniform Curve Sprite Scale=DataInterface() {propertyValues {Curve {keys [
        {interpMode=RCIM_Cubic, tangentMode=RCTM_User, time=0, value=0}
        {interpMode=RCIM_Cubic, tangentMode=RCTM_User, time=0.5, value=1}
        {tangentMode=RCTM_User, time=1, value=0}
       ]}}}
   ) {module=ScaleSpriteSize001}
   &ScaleColor(
    Scale Alpha=&Multiply_Float(
     A=&Multiply_Float(
      A=$System.TODScalar
      B=&Multiply_Float(
       A=&FloatFromCurve(
        FloatCurve=DataInterface() {propertyValues {Curve {keys [
            {arriveTangent=6.5707, interpMode=RCIM_Cubic, leaveTangent=6.5707, tangentMode=RCTM_User, time=0, value=0}
            {interpMode=RCIM_Cubic, tangentMode=RCTM_User, time=0.3, value=1}
            {tangentMode=RCTM_User, time=1, value=0}
           ]}}}
        Scale Curve=&RandomRangeFloat(Minimum=0.1, Maximum=0.5)
       )
       B=1
      )
     )
     B=&Lerp_Float(A=0, B=1, Alpha=$System.AudioReactivity)
    )
    Scale RGB=&Multiply_Vector(
     A=&VectorFromFloat(Value=&Multiply_Float(
       A=&RandomRangeFloat(Minimum=0.75, Maximum=1)
       B=&SelectFloatFromArray(Direct Array Index=$Engine.QualityLevel, Float Selection Array=DataInterface() {propertyValues {FloatData [10,5,1,1,1]}})
      ))
     B=&MakeVectorFromLinearColorRGB(Color=$System.TODColor)
    )
    ScaleA=true
    ScaleRGB=true
   )
   &ScaleSpriteSize(
    Scale Sprite Size Mode=::NewEnumerator0
    Uniform Scale Factor=&MakeCustomFloatFromBool(
     Bool=$User.bisHandHidden
     True Float=&SelectFloatFromArray(
      Direct Array Index=&SpawnGroup()
      Float Selection Array=DataInterface() {propertyValues {FloatData [1,1,1,1,0,0,1,1]}}
     )
     False Float=1
    )
   )
   &ScaleSpriteSize(
    Scale Sprite Size Mode=::NewEnumerator0
    Uniform Scale Factor=&MakeCustomFloatFromBool(
     Bool=$User.bIsFootHidden
     True Float=&SelectFloatFromArray(
      Direct Array Index=&SpawnGroup()
      Float Selection Array=DataInterface() {propertyValues {FloatData [0,0,1,1,1,1,1,1]}}
     )
     False Float=1
    )
   ) {module=ScaleSpriteSize002}
   /SetVariables(
    Particles.CenterFlareColor:LinearColor=&MultiplyLinearColor(
     Linear Color=$Particles.Color
     Scale Factor=&MakeVector4FromVectorAndFloat(
      "Vector (RGB)"=&MakeVectorFromLinearColorRGB(Color=<rgba 2.4876 3.9656 5 1>)
      "Float (Alpha)"=&Multiply_Float(
       A=&FloatFromCurve(
        FloatCurve=DataInterface() {propertyValues {Curve {
           defaultValue=1
           keys [
            {arriveTangent=4.0437, interpMode=RCIM_Cubic, leaveTangent=4.0437, tangentMode=RCTM_User, time=0, value=0.1}
            {arriveTangent=-0.8452, interpMode=RCIM_Cubic, leaveTangent=-0.8452, tangentMode=RCTM_User, time=0.1418, value=0.9847}
            {arriveTangent=-0.8544, interpMode=RCIM_Cubic, leaveTangent=-0.8544, tangentMode=RCTM_User, time=1, value=0}
           ]
          }}}
        Scale Curve=&Multiply_Float(A=2, B=$System.VelocityFade)
       )
       B=&RemapRange(
        Input Value To Remap=&DistanceBetweenPositions(A=$Engine.Emitter.SimulationPosition, B=$Particles.Position)
        Input Min=0
        Input Max=300
        Output Min=1
        Output Max=0
       )
      )
     )
    )
    Particles.CenterFlareSize:Vector2f=&Multiply_Vector2(A=$Particles.SpriteSize, B=&Vector2DFromFloat(Value=0.5))
    Particles.CoronaFlareColor:LinearColor=&MultiplyLinearColor(
     Linear Color=$Particles.Color
     Scale Factor=&MakeVector4FromVectorAndFloat(
      "Vector (RGB)"=&MakeVectorFromLinearColorRGB(Color=<rgba 0.0411 0.0805 0.15 1>)
      "Float (Alpha)"=&Multiply_Float(
       A=&FloatFromCurve(
        FloatCurve=DataInterface() {propertyValues {Curve {
           defaultValue=1
           keys [
            {arriveTangent=4.0437, interpMode=RCIM_Cubic, leaveTangent=4.0437, tangentMode=RCTM_User, time=0, value=0.1}
            {arriveTangent=-0.8452, interpMode=RCIM_Cubic, leaveTangent=-0.8452, tangentMode=RCTM_User, time=0.1418, value=0.9847}
            {arriveTangent=-0.8544, interpMode=RCIM_Cubic, leaveTangent=-0.8544, tangentMode=RCTM_User, time=1, value=0}
           ]
          }}}
        Scale Curve=&Multiply_Float(A=1, B=$System.VelocityFade)
       )
       B=&RemapRange(
        Input Value To Remap=&DistanceBetweenPositions(A=$Engine.Emitter.SimulationPosition, B=$Particles.Position)
        Input Min=0
        Input Max=300
        Output Min=1
        Output Max=0
       )
      )
     )
    )
    Particles.CoronaFlareSize:Vector2f=&Multiply_Vector2(
     A=&Multiply_Vector2D_ByFloat(Float=&Lerp_Float(A=0.3, B=1, Alpha=$System.AudioReactivity), Vector2D=$Particles.Initial.SpriteSize)
     B=&Vector2DFromFloat(Value=0.75)
    )
   ) {isSetParameters=true}
   &CurlNoiseForce(Bias Noise=false, Noise Frequency=15, Noise Strength=44.1779, Pan Noise=false)
   &Drag(Use Linear Drag=true, Use Rotational Drag=true)
   &SolveForcesAndVelocity(Clamp Velocity=false, Limit Acceleration=false)
   &ScaleColor(Scale Alpha=$System.VelocityFade, ScaleA=true, ScaleRGB=true) {module=ScaleColor001}
  ]
  rendererClasses [&NiagaraSpriteRendererProperties]
  renderers [
   {rendererClass=&NiagaraSpriteRendererProperties}
   {rendererClass=&NiagaraSpriteRendererProperties, rendererIndex=1}
  ]
 }
 Swirls {
  emitterUpdateScript [
   &EmitterState(Inactive Response=::NewEnumerator1, Life Cycle Mode=::NewEnumerator1, UseLoopDelay=false)
   &SpawnRate(Spawn Probability=$System.AudioReactivity, SpawnRate=&Multiply_Float(A=24, B=1), Use Spawn Probability=true)
  ]
  particleSpawnScript [
   &FN_SetSpawnGroup(Number of Spawn Groups=16)
   &InitializeParticle(
    Color=<rgba 1 1 1 0.5>
    Lifetime=&SelectFloatFromArray(
     Direct Array Index=&SpawnGroup()
     Float Selection Array=DataInterface() {propertyValues {FloatData [1.7,2,1.7,3,1.7,1.7,1.7,1.7,1.7,1.7,1.7,1.7,2,1.7,1.7]}}
    )
    Mesh Scale=&Multiply_VectorByFloat(
     Vector=&SelectVectorFromArray(
      Direct Array Index=&SpawnGroup()
      Vector Selection Array=DataInterface() {propertyValues {FloatData [
         <1 1 1>
         <0.8 1 0.8>
         <1 1 1>
         <1 1 1>
         <1 1 1>
         <1 0.8 0.8>
         <1 1 1>
         <1 1 1>
         <1.3 1 1>
         <0.8 0.8 0.8>
         <1 1 1>
         <1 1 1>
         <0.7 0.7 0.7>
         <1 1 1>
         <0.7 0.7 0.7>
         <0.7 0.7 0.7>
        ]}}
     )
     Float=1
    )
    Mesh Scale Mode=::NewEnumerator4
    UsePositionOffset=false
    Write Material Random=true
   )
   &InitialMeshOrientation(
    Mesh Orientation Mode=::NewEnumerator2
    Rotation=&SelectVectorFromArray(
     Direct Array Index=&SpawnGroup()
     Vector Selection Array=DataInterface() {propertyValues {FloatData [
        <0 0.25 0>
        <0 0.25 0>
        <0 0.25 0>
        <0 0.25 0>
        <0.027 0.0982 0.0655>
        <0.1441 0.0991 0.036>
        <0.3153 0.1712 0.3333>
        <0.1892 -0.0541 -0.1532>
        <-0.2432 0.009 -0.0811>
        <0 -0.2523 0>
        <-0.0541 0.1171 0.0541>
        <0 -0.1892 0>
        <0 -0.2703 0>
        <0 -0.2523 0>
        <0.2252 -0.0811 0.0901>
        <0.1532 -0.1892 0>
       ]}}
    )
    Use Rotation Vector=true
   ) {module=InitialMeshOrientation001}
   /SetVariables(
    Particles.CustomRandom:NiagaraFloat=&RandomRangeFloat(Minimum=0, Maximum=1)
    Particles.SpawnEmitterAge:NiagaraFloat=$Emitter.Age
   ) {isSetParameters=true}
   &CameraOffset(Camera Offset Amount=8)
   /SetVariables(Particles.MeshIndex=&SelectIntFromArray(
     Direct Array Index=&SpawnGroup()
     Int Selection Array=DataInterface() {propertyValues {IntData [0,1,2,3,4,5,6,7,7,2,4,0,1,2,2,6]}}
    )) {isSetParameters=true}
  ]
  particleUpdateScript [
   /SetVariables(Transient.Shared.FloatFromCurve_FloatCurve=DataInterface() {propertyValues {Curve {
       defaultValue=1
       keys [<at 0 -0.3 RCIM_Cubic>, {arriveTangent=0.6523, interpMode=RCIM_Cubic, leaveTangent=0.6523, tangentMode=RCTM_User, time=1, value=0.3}]
      }}}) {isSetParameters=true}
   &ParticleState()
   &ScaleMeshSize(Scale Factor=&VectorFromFloat(Value=&MakeCustomFloatFromBool(
      Bool=$User.bisHandHidden
      True Float=&SelectFloatFromArray(
       Direct Array Index=&SpawnGroup()
       Float Selection Array=DataInterface() {propertyValues {FloatData [1,1,1,1,0,1,1,0,1,0,1,1,1,1,1,1]}}
      )
      False Float=1
     )))
   &ScaleColor(
    Scale Alpha=&MakeCustomFloatFromBool(
     Bool=$User.bisHandHidden
     True Float=&SelectFloatFromArray(
      Direct Array Index=&SpawnGroup()
      Float Selection Array=DataInterface() {propertyValues {FloatData [1,1,1,1,0,1,1,0,1,0,1,1,1,1,1,1]}}
     )
     False Float=1
    )
    ScaleA=true
    ScaleRGB=false
   ) {module=ScaleColor001}
   &ScaleMeshSize(Scale Factor=&VectorFromFloat(Value=&MakeCustomFloatFromBool(
      Bool=$User.bIsFootHidden
      True Float=&SelectFloatFromArray(
       Direct Array Index=&SpawnGroup()
       Float Selection Array=DataInterface() {propertyValues {FloatData [0,0,1,1,1,1,1,1,1,1,1,0,0,1,1,1]}}
      )
      False Float=1
     ))) {module=ScaleMeshSize001}
   &ScaleColor(
    Scale Alpha=&MakeCustomFloatFromBool(
     Bool=$User.bIsFootHidden
     True Float=&SelectFloatFromArray(
      Direct Array Index=&SpawnGroup()
      Float Selection Array=DataInterface() {propertyValues {FloatData [0,0,1,1,1,1,1,1,1,1,1,0,0,1,1,1]}}
     )
     False Float=1
    )
    ScaleA=true
    ScaleRGB=false
   ) {module=ScaleColor002}
   &DynamicMaterialParameters(
    Index 0 Param 1=&Lerp_Float(
     A=&FloatFromCurve(
      FloatCurve=?
      Transient.Shared.FloatFromCurve_FloatCurve:NiagaraDataInterfaceCurve=DataInterface() {propertyValues {Curve {
         defaultValue=1
         keys [<at 0 -0.3 RCIM_Cubic>, {arriveTangent=0.6523, interpMode=RCIM_Cubic, leaveTangent=0.6523, tangentMode=RCTM_User, time=1, value=0.3}]
        }}}
     )
     B=0
     Alpha=0
    )
    Index 0 Param 2=&Multiply_Float(
     A=&SelectFloatFromArray(
      Direct Array Index=&SpawnGroup()
      Float Selection Array=DataInterface() {propertyValues {FloatData [3.5,3.5,3.5,2.5,3.5,3.5,3.5,3.5,3.5,3.5,3.5,3.5,3.5,3.5]}}
     )
     B=&Lerp_Float(A=0.3, B=1, Alpha=$System.AudioReactivity)
    )
    Index 0 Param 3=1
    Module.Index0Param0Write=true
    Module.Index0Param1Write=true
    Module.Index0Param2Write=true
    Module.Index0Param3Write=true
   )
   &ScaleColor(
    Scale Alpha=&FloatFromCurve(
     FloatCurve=DataInterface() {propertyValues {Curve {keys [{arriveTangent=-1.5781, interpMode=RCIM_Cubic, leaveTangent=-1.5781, tangentMode=RCTM_User, time=0.5, value=1}, <at 1 0 RCIM_Cubic>]}}}
     Scale Curve=&Multiply_Float(A=1, B=$System.AudioReactivity)
    )
    Scale RGB=&MakeVectorFromLinearColorRGB(Color=&Multiply_LinearColorByFloat(
      Linear Color=<rgba 0.261 1.1279 1.6 1>
      Float=&Waveform("Waveform [1]"=::NewEnumerator0, "[1] Frequency"=2, "[1] Amplitude Min/Max"=<1 1.4>)
     ))
    ScaleA=true
    ScaleRGB=true
   )
   &SocketLocation(
    SamplePosOffset=true
    Sampled Position Offset=&SelectVectorFromArray(
     Direct Array Index=&SpawnGroup()
     Vector Selection Array=DataInterface() {propertyValues {FloatData [
        <-27.7571 -9.9251 -2.0223>
        <-37.7658 5 2>
        <-40 3 -1>
        <-2.4661 12 0>
        <-23.619 3.9936 -1.1866>
        <-22.5942 2.7183 -0.3248>
        <17.5288 -12.9719 3.7263>
        <-17.7027 -0.6842 -0.6556>
        <11.5674 -0.7417 -13.009>
        <22.8651 -0.9516 -1.02>
        <2.6341 1.4443 -1.8696>
        <28.9925 -4.1423 0>
        <35.8277 10.8332 0.6847>
        <39.2207 5.0166 0>
        <7.2287 5.4667 8.1216>
        <24.2993 8.7708 -0.6667>
       ]}}
    )
    Socket Index=&SelectIntFromArray(
     Direct Array Index=&SpawnGroup()
     Int Selection Array=DataInterface() {propertyValues {IntData [0,0,1,2,3,4,6,3,7,8,9,5,5,6,7,9]}}
    )
    Socket Reader=$System.Socket Reader
    Socket Sampling Mode=::NewEnumerator3
    VelocitySpeedLimit=false
   )
  ]
  rendererClasses [&NiagaraMeshRendererProperties]
  renderers [{rendererClass=&NiagaraMeshRendererProperties, rendererIndex=0}]
 }
}
summary {
 systemName=NS_Character_SprintLuckSlaw
 emitters [
  {emitterName=Stars, rendererClasses [&NiagaraSpriteRendererProperties]}
  {emitterName=Swirls, rendererClasses [&NiagaraMeshRendererProperties]}
  {emitterName=Glow, rendererClasses [&NiagaraSpriteRendererProperties]}
 ]
 userVariables [
  User.bIsADS {type=NiagaraBool}
  User.bIsFootHidden {type=NiagaraBool}
  User.bIsFrontEnd {type=NiagaraBool}
  User.bIsFrontEndPreview {type=NiagaraBool}
  User.bisHandHidden {type=NiagaraBool}
  User.DebugReactivity {type=NiagaraBool}
  User.EnableFrontendFakeAudioLoop {type=NiagaraFloat, defaultValue=0, description=""}
  User.IsSkydiving {type=NiagaraBool}
 ]
}
