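# Material function for the SprintLuckSlaw character body that adds audio-reactive gem glow and rim lighting to incoming material attributes.
# MF_AudioReactivity amplitude and spectrum, blendable with a fake frontend loop, drive TechArt-texture gem masks and tinted emissive.
# Fresnel rim light is confined by pre-skinned local-position sphere masks; base color and specular are also adjusted.
# The custom synthwave HLSL node feeds only an unused clamp; the wired fake amplitude is a time sawtooth.
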MaterialFunction /BRCosmetics/Characters/Player/Female/Medium/Bodies/F_MED_SprintLuckSlaw/Materials/MF_SprintLuckSlaw
graph {expressions {
  MaterialExpressionAdd_0 {
   class=MaterialExpressionAdd
   inputs [
    A {from=MaterialExpressionMultiply_0}
    B {from=MaterialExpressionPower_0}
   ]
  }
  MaterialExpressionAdd_1 {
   class=MaterialExpressionAdd
   inputs [
    A {from=MaterialExpressionNamedRerouteDeclaration_11}
    B {from=MaterialExpressionMultiply_10}
   ]
  }
  MaterialExpressionAdd_14 {
   class=MaterialExpressionAdd
   inputs [
    A {from=MaterialExpressionLinearInterpolate_17}
    B {from=MaterialExpressionMultiply_20}
   ]
  }
  MaterialExpressionAdd_3 {
   class=MaterialExpressionAdd
   inputs [
    A {from=MaterialExpressionTextureSample_0, output=G}
    B {from=MaterialExpressionFrac_2}
   ]
  }
  MaterialExpressionAdd_5 {
   class=MaterialExpressionAdd
   inputs [
    A {from=MaterialExpressionAdd_3}
    B {from=""}
   ]
   properties {constB=0.5}
  }
  MaterialExpressionClamp_2 {
   class=MaterialExpressionClamp
   inputs [
    None {from=MaterialExpressionCustom_0, output=Spectrum}
    Min {from="", output=""}
    Max {from="", output=""}
   ]
   properties {minDefault=0.5}
  }
  MaterialExpressionColorRamp_0 {
   class=MaterialExpressionColorRamp
   inputs [None {from=MaterialExpressionTextureSample_0, output=G}]
   properties {colorCurve="<self>.ColorCurve"}
  }
  MaterialExpressionComponentMask_5 {class=MaterialExpressionComponentMask, inputs [None {from=MaterialExpressionLocalPosition_0, output=XYZ}], properties {r=true}}
  MaterialExpressionComponentMask_6 {class=MaterialExpressionComponentMask, inputs [None {from=MaterialExpressionLocalPosition_0, output=XYZ}], properties {g=true}}
  MaterialExpressionConstant3Vector_1 {class=MaterialExpressionConstant3Vector, properties {constant=<rgba 16 -5 50 0>}}
  MaterialExpressionConstant3Vector_2 {class=MaterialExpressionConstant3Vector, properties {constant=<rgba -1 1 1 0>}}
  MaterialExpressionConstant3Vector_3 {class=MaterialExpressionConstant3Vector, properties {constant=<rgba 50 0 130 0>}}
  MaterialExpressionConstant3Vector_4 {class=MaterialExpressionConstant3Vector, properties {constant=<rgba 0 0 105 0>}}
  MaterialExpressionConstant3Vector_6 {class=MaterialExpressionConstant3Vector, properties {constant=<rgba 0 6 132 0>}}
  MaterialExpressionConstant3Vector_7 {class=MaterialExpressionConstant3Vector, properties {constant=<rgba 1 -1.5 1.015 0>}}
  MaterialExpressionConstant_0 {class=MaterialExpressionConstant, properties {r=1}}
  MaterialExpressionConstant_1 {class=MaterialExpressionConstant}
  MaterialExpressionConvert_0 {
   class=MaterialExpressionConvert
   inputs [None {from=MaterialExpressionNamedRerouteUsage_15}]
   properties {
    nodeName=Break Float4
    convertInputs [{
      type=Vector4
      defaultValue=<rgba 0 0 0 1>
      expressionInput {
       expression=&self:MaterialExpressionNamedRerouteUsage_15
       inputName=None
       mask=0
       maskA=0
       maskB=0
       maskG=0
       maskR=0
       outputIndex=0
      }
     }]
    convertOutputs [{type=Scalar}, {type=Scalar}, {type=Scalar}, {type=Scalar}]
   }
  }
  MaterialExpressionConvert_1 {
   class=MaterialExpressionConvert
   inputs [
    None {from=MaterialExpressionColorRamp_0}
    None {from=MaterialExpressionTextureSample_0, output=B}
   ]
   properties {
    nodeName=Make Vector4
    convertInputs [
     {
      type=Vector4
      defaultValue=<rgba 0 0 0 1>
      expressionInput {
       expression=&self:MaterialExpressionColorRamp_0
       inputName=None
       mask=0
       maskA=0
       maskB=0
       maskG=0
       maskR=0
       outputIndex=0
      }
     }
     {
      type=Scalar
      expressionInput {
       expression=&self:MaterialExpressionTextureSample_0
       inputName=None
       mask=1
       maskA=0
       maskB=1
       maskG=0
       maskR=0
       outputIndex=3
      }
     }
    ]
    convertOutputs [{type=Vector4, defaultValue=<rgba 0 0 0 1>}]
   }
  }
  MaterialExpressionCustom_0 {
   class=MaterialExpressionCustom
   inputs [
    Time {from=MaterialExpressionTime_2, output=""}
    Speed {from=MaterialExpressionScalarParameter_7, output=""}
   ]
   properties {
    code="// Fake Synthwave Audio (rhythmic idle)\r\nfloat t    = Time * Speed;\r\nfloat beat = t;\r\n\r\n// Sub-beat clocks (all \"0..1 then snap to 0 again\")\r\nfloat kickPhase      = frac(beat);\r\nfloat snarePhase     = frac((beat - 1.0) * 0.5) * 2.0;\r\nfloat eighthPhase    = frac(beat * 2.0);\r\nfloat sixteenthPhase = frac(beat * 4.0);\r\n\r\n// 16th-note index, wrapped to position within a 4-beat bar (0..15)\r\nfloat sixteenthIdx   = floor(beat * 4.0);\r\nfloat sixteenthInBar = sixteenthIdx - floor(sixteenthIdx / 16.0) * 16.0;\r\n\r\n// DRUM HITS\r\nfloat kick  = pow(saturate(1.0 - kickPhase),      6.0);  // every beat\r\nfloat snare = pow(saturate(1.0 - snarePhase),     7.0);  // backbeats (2, 4)\r\nfloat hat   = pow(saturate(1.0 - eighthPhase),    8.0) * 0.6;\r\n\r\n// ARPEGGIO (16th notes, repeating melodic shape per bar)\r\nfloat arpShape = (sin(sixteenthInBar * 0.78) + sin(sixteenthInBar * 1.43)) * 0.25 + 0.5;\r\nfloat arpEnv   = pow(saturate(1.0 - sixteenthPhase), 6.0);\r\nfloat arp      = arpEnv * arpShape;\r\n\r\n// BASS LINE (8th notes, varying velocity per note)\r\nfloat bassNote    = floor(beat * 2.0);\r\nfloat bassPattern = sin(bassNote * 1.1) * 0.3 + 0.7;   // 0.4..1.0\r\nfloat bassEnv     = pow(saturate(1.0 - eighthPhase), 4.0);\r\nfloat bassLine    = bassEnv * bassPattern;\r\n\r\n// BUILD SPECTRUM\r\nfloat4 spec;\r\nspec.r = kick  * 0.30 + bassLine * 0.15 + 0.06;          // Bass: kick + bassline\r\nspec.g = bassLine * 0.70 + kick  * 0.25 + 0.08;          // Low:  bassline body + kick body\r\nspec.b = snare * 0.75 + arp  * 0.85 + 0.08;              // Mid:  snare + arp\r\nspec.a = hat   * 0.75 + arp  * 0.35 + 0.05;              // High: hats + arp shimmer\r\nSpectrum = saturate(spec);\r\n\r\n// AMPLITUDE — clear pulses, low floor (no more \"breathing\")\r\nfloat pulse = max(max(kick, snare * 0.85), max(hat * 0.4, arp * 0.35));\r\nfloat amp   = saturate(0.20 + pulse * 0.75);\r\n\r\nreturn amp;"
    description=Fake Synthwave Audio
    outputType=CMOT_Float1
    showCode=true
    additionalOutputs [{outputName=Spectrum, outputType=CMOT_Float4}]
    inputs [
     {
      inputName=Time
      input {
       expression=&self:MaterialExpressionTime_2
       inputName=None
       mask=0
       maskA=0
       maskB=0
       maskG=0
       maskR=0
       outputIndex=0
      }
     }
     {
      inputName=Speed
      input {
       expression=&self:MaterialExpressionScalarParameter_7
       inputName=None
       mask=0
       maskA=0
       maskB=0
       maskG=0
       maskR=0
       outputIndex=0
      }
     }
    ]
   }
  }
  MaterialExpressionDesaturation_0 {
   class=MaterialExpressionDesaturation
   inputs [
    None {from=MaterialExpressionNamedRerouteUsage_23, output=""}
    Fraction {from=MaterialExpressionConstant_0, output=""}
   ]
   properties {}
  }
  MaterialExpressionDotProduct_1 {
   class=MaterialExpressionDotProduct
   inputs [
    A {from=MaterialExpressionPower_5, output=""}
    B {from=MaterialExpressionMultiply_1, output=""}
   ]
   properties {}
  }
  MaterialExpressionFrac_0 {class=MaterialExpressionFrac, inputs [None {from=MaterialExpressionMultiply_56, output=""}], properties {}}
  MaterialExpressionFrac_2 {class=MaterialExpressionFrac, inputs [None {from=MaterialExpressionMultiply_12, output=""}], properties {}}
  MaterialExpressionFresnel_0 {
   class=MaterialExpressionFresnel
   inputs [
    ExponentIn {from="", output=""}
    BaseReflectFractionIn {from="", output=""}
    Normal {from="", output=""}
   ]
   properties {exponent=0.8}
  }
  MaterialExpressionFresnel_1 {
   class=MaterialExpressionFresnel
   inputs [
    ExponentIn {from="", output=""}
    BaseReflectFractionIn {from="", output=""}
    Normal {from="", output=""}
   ]
   properties {exponent=3}
  }
  MaterialExpressionFresnel_2 {
   class=MaterialExpressionFresnel
   inputs [
    ExponentIn {from="", output=""}
    BaseReflectFractionIn {from="", output=""}
    Normal {from="", output=""}
   ]
   properties {exponent=0.2}
  }
  MaterialExpressionFunctionInput_0 {
   class=MaterialExpressionFunctionInput
   inputs [Preview {from="", output=""}]
   properties {inputName=OUT_MA, inputType=FunctionInput_MaterialAttributes}
  }
  MaterialExpressionFunctionInput_4 {
   class=MaterialExpressionFunctionInput
   inputs [Preview {from="", output=""}]
   properties {bUsePreviewValueAsDefault=true, inputName=PreSkinnedLocation, inputType=FunctionInput_Scalar, sortPriority=2, previewValue=<1 1 1 1>}
  }
  MaterialExpressionFunctionOutput_1 {
   class=MaterialExpressionFunctionOutput
   inputs [None {from=MaterialExpressionSetMaterialAttributes_0, output=""}]
   properties {outputName=OUT_MA}
  }
  MaterialExpressionGetMaterialAttributes_0 {
   class=MaterialExpressionGetMaterialAttributes
   inputs [None {from=MaterialExpressionFunctionInput_0, output=""}]
   properties {attributeGetTypes [#1, #2, #3]}
  }
  MaterialExpressionLinearInterpolate_0 {
   class=MaterialExpressionLinearInterpolate
   inputs [
    A {from=""}
    B {from=""}
    Alpha {from=MaterialExpressionNamedRerouteUsage_3}
   ]
   properties {constA=15, constB=30}
  }
  MaterialExpressionLinearInterpolate_1 {
   class=MaterialExpressionLinearInterpolate
   inputs [
    A {from=""}
    B {from=""}
    Alpha {from=MaterialExpressionNamedRerouteUsage_3}
   ]
   properties {constA=12, constB=32}
  }
  MaterialExpressionLinearInterpolate_17 {
   class=MaterialExpressionLinearInterpolate
   inputs [
    A {from=MaterialExpressionNamedRerouteDeclaration_11}
    B {from=MaterialExpressionAdd_1}
    Alpha {from=MaterialExpressionMultiply_21}
   ]
  }
  MaterialExpressionLinearInterpolate_18 {
   class=MaterialExpressionLinearInterpolate
   inputs [
    A {from=MaterialExpressionNamedRerouteUsage_23}
    B {from=MaterialExpressionLinearInterpolate_24}
    Alpha {from=MaterialExpressionMultiply_18}
   ]
  }
  MaterialExpressionLinearInterpolate_24 {
   class=MaterialExpressionLinearInterpolate
   inputs [
    A {from=MaterialExpressionNamedRerouteUsage_23}
    B {from=MaterialExpressionMultiply_53}
    Alpha {from=MaterialExpressionNamedRerouteUsage_38}
   ]
   properties {constB=0}
  }
  MaterialExpressionLinearInterpolate_25 {
   class=MaterialExpressionLinearInterpolate
   inputs [
    A {from=MaterialExpressionMultiply_41}
    B {from=MaterialExpressionMultiply_38}
    Alpha {from=MaterialExpressionPower_7}
   ]
   properties {constA=1}
  }
  MaterialExpressionLinearInterpolate_26 {
   class=MaterialExpressionLinearInterpolate
   inputs [
    A {from=MaterialExpressionConstant3Vector_1}
    B {from=MaterialExpressionMultiply_42}
    Alpha {from=MaterialExpressionStep_9}
   ]
  }
  MaterialExpressionLinearInterpolate_27 {
   class=MaterialExpressionLinearInterpolate
   inputs [
    A {from=MaterialExpressionConstant3Vector_3}
    B {from=MaterialExpressionMultiply_44}
    Alpha {from=MaterialExpressionStep_9}
   ]
  }
  MaterialExpressionLinearInterpolate_28 {
   class=MaterialExpressionLinearInterpolate
   inputs [
    A {from=MaterialExpressionConstant3Vector_6}
    B {from=MaterialExpressionMultiply_48}
    Alpha {from=MaterialExpressionStep_10}
   ]
  }
  MaterialExpressionLinearInterpolate_29 {
   class=MaterialExpressionLinearInterpolate
   inputs [
    A {from=""}
    B {from=""}
    Alpha {from=MaterialExpressionStep_10}
   ]
   properties {constA=6, constB=4}
  }
  MaterialExpressionLinearInterpolate_3 {
   class=MaterialExpressionLinearInterpolate
   inputs [
    A {from=MaterialExpressionGetMaterialAttributes_0, output=Specular}
    B {from=""}
    Alpha {from=MaterialExpressionNamedRerouteUsage_38}
   ]
   properties {constB=0.1}
  }
  MaterialExpressionLinearInterpolate_6 {
   class=MaterialExpressionLinearInterpolate
   inputs [
    A {from=MaterialExpressionMaterialFunctionCall_1, output=Spectrum}
    B {from=MaterialExpressionNamedRerouteUsage_18}
    Alpha {from=MaterialExpressionScalarParameter_4}
   ]
  }
  MaterialExpressionLinearInterpolate_7 {
   class=MaterialExpressionLinearInterpolate
   inputs [
    A {from=MaterialExpressionMaterialFunctionCall_1, output=Amplitude}
    B {from=MaterialExpressionNamedRerouteUsage_17}
    Alpha {from=MaterialExpressionScalarParameter_4}
   ]
  }
  MaterialExpressionLocalPosition_0 {class=MaterialExpressionLocalPosition, inputs [], properties {localOrigin=InstancePreSkinning}}
  MaterialExpressionMaterialFunctionCall_1 {
   class=MaterialExpressionMaterialFunctionCall
   inputs []
   properties {materialFunction=/Game/Packages/Fortress_SharedMaterials/Base_Material_Functions/BaseCharacterShaderFunctions/Audio/MF_AudioReactivity}
  }
  MaterialExpressionMax_0 {
   class=MaterialExpressionMax
   inputs [
    A {from=MaterialExpressionSine_0}
    B {from=MaterialExpressionSine_2}
   ]
  }
  MaterialExpressionMax_1 {
   class=MaterialExpressionMax
   inputs [
    A {from=MaterialExpressionDotProduct_1}
    B {from=MaterialExpressionMultiply_19}
   ]
  }
  MaterialExpressionMax_3 {
   class=MaterialExpressionMax
   inputs [
    A {from=MaterialExpressionMultiply_34}
    B {from=MaterialExpressionMax_1}
   ]
  }
  MaterialExpressionMax_6 {
   class=MaterialExpressionMax
   inputs [
    A {from=MaterialExpressionSphereMask_0}
    B {from=MaterialExpressionSphereMask_1}
   ]
  }
  MaterialExpressionMax_7 {
   class=MaterialExpressionMax
   inputs [
    A {from=MaterialExpressionMax_6}
    B {from=MaterialExpressionMax_8}
   ]
  }
  MaterialExpressionMax_8 {
   class=MaterialExpressionMax
   inputs [
    A {from=MaterialExpressionSphereMask_2}
    B {from=MaterialExpressionSphereMask_3}
   ]
  }
  MaterialExpressionMultiply_0 {
   class=MaterialExpressionMultiply
   inputs [
    A {from=""}
    B {from=MaterialExpressionPower_0}
   ]
  }
  MaterialExpressionMultiply_1 {
   class=MaterialExpressionMultiply
   inputs [
    A {from=MaterialExpressionConvert_1}
    B {from=MaterialExpressionNamedRerouteDeclaration_32}
   ]
  }
  MaterialExpressionMultiply_10 {
   class=MaterialExpressionMultiply
   inputs [
    A {from=MaterialExpressionMultiply_23}
    B {from=""}
   ]
   properties {constB=10}
  }
  MaterialExpressionMultiply_12 {
   class=MaterialExpressionMultiply
   inputs [
    A {from=MaterialExpressionTime_1}
    B {from=""}
   ]
  }
  MaterialExpressionMultiply_15 {
   class=MaterialExpressionMultiply
   inputs [
    A {from=MaterialExpressionMultiply_8}
    B {from=""}
   ]
   properties {constB=0.2}
  }
  MaterialExpressionMultiply_16 {
   class=MaterialExpressionMultiply
   inputs [
    A {from=MaterialExpressionStep_7}
    B {from=MaterialExpressionStep_8}
   ]
  }
  MaterialExpressionMultiply_18 {
   class=MaterialExpressionMultiply
   inputs [
    A {from=MaterialExpressionNamedRerouteUsage_24}
    B {from=MaterialExpressionScalarParameter_0}
   ]
  }
  MaterialExpressionMultiply_19 {
   class=MaterialExpressionMultiply
   inputs [
    A {from=MaterialExpressionMultiply_15}
    B {from=MaterialExpressionNamedRerouteUsage_22}
   ]
   properties {constB=0.25}
  }
  MaterialExpressionMultiply_2 {
   class=MaterialExpressionMultiply
   inputs [
    A {from=MaterialExpressionNamedRerouteUsage_35}
    B {from=MaterialExpressionVectorParameter_0, output=RGB}
   ]
  }
  MaterialExpressionMultiply_20 {
   class=MaterialExpressionMultiply
   inputs [
    A {from=MaterialExpressionNamedRerouteUsage_42}
    B {from=MaterialExpressionScalarParameter_0}
   ]
  }
  MaterialExpressionMultiply_21 {
   class=MaterialExpressionMultiply
   inputs [
    A {from=MaterialExpressionNamedRerouteUsage_21}
    B {from=MaterialExpressionScalarParameter_0}
   ]
  }
  MaterialExpressionMultiply_23 {
   class=MaterialExpressionMultiply
   inputs [
    A {from=MaterialExpressionNamedRerouteUsage_38}
    B {from=MaterialExpressionPower_3}
   ]
   properties {constB=4}
  }
  MaterialExpressionMultiply_34 {
   class=MaterialExpressionMultiply
   inputs [
    A {from=MaterialExpressionMultiply_16}
    B {from=""}
   ]
   properties {constB=0.8}
  }
  MaterialExpressionMultiply_35 {
   class=MaterialExpressionMultiply
   inputs [
    A {from=MaterialExpressionMax_3}
    B {from=MaterialExpressionLinearInterpolate_25}
   ]
  }
  MaterialExpressionMultiply_38 {
   class=MaterialExpressionMultiply
   inputs [
    A {from=MaterialExpressionNamedRerouteUsage_36}
    B {from=MaterialExpressionVectorParameter_0, output=RGB}
   ]
  }
  MaterialExpressionMultiply_41 {
   class=MaterialExpressionMultiply
   inputs [
    A {from=MaterialExpressionMax_1}
    B {from=MaterialExpressionMultiply_2}
   ]
   properties {constB=0.5}
  }
  MaterialExpressionMultiply_42 {
   class=MaterialExpressionMultiply
   inputs [
    A {from=MaterialExpressionConstant3Vector_1}
    B {from=MaterialExpressionConstant3Vector_2}
   ]
  }
  MaterialExpressionMultiply_43 {
   class=MaterialExpressionMultiply
   inputs [
    A {from=MaterialExpressionShaderStageSwitch_0}
    B {from=MaterialExpressionMultiply_45}
   ]
  }
  MaterialExpressionMultiply_44 {
   class=MaterialExpressionMultiply
   inputs [
    A {from=MaterialExpressionConstant3Vector_3}
    B {from=MaterialExpressionConstant3Vector_2}
   ]
  }
  MaterialExpressionMultiply_45 {
   class=MaterialExpressionMultiply
   inputs [
    A {from=MaterialExpressionPower_8}
    B {from=""}
   ]
   properties {constB=8}
  }
  MaterialExpressionMultiply_46 {
   class=MaterialExpressionMultiply
   inputs [
    A {from=MaterialExpressionVectorParameter_0, output=RGB}
    B {from=MaterialExpressionMultiply_43}
   ]
  }
  MaterialExpressionMultiply_47 {
   class=MaterialExpressionMultiply
   inputs [
    A {from=MaterialExpressionMultiply_46}
    B {from=MaterialExpressionNamedRerouteUsage_41}
   ]
  }
  MaterialExpressionMultiply_48 {
   class=MaterialExpressionMultiply
   inputs [
    A {from=MaterialExpressionConstant3Vector_6}
    B {from=MaterialExpressionConstant3Vector_7}
   ]
  }
  MaterialExpressionMultiply_53 {
   class=MaterialExpressionMultiply
   inputs [
    A {from=MaterialExpressionDesaturation_0}
    B {from=""}
   ]
   properties {constB=0.75}
  }
  MaterialExpressionMultiply_54 {
   class=MaterialExpressionMultiply
   inputs [
    A {from=MaterialExpressionTime_2}
    B {from=MaterialExpressionScalarParameter_7}
   ]
  }
  MaterialExpressionMultiply_55 {
   class=MaterialExpressionMultiply
   inputs [
    A {from=MaterialExpressionSubtract_1}
    B {from=""}
   ]
   properties {constB=0.5}
  }
  MaterialExpressionMultiply_56 {
   class=MaterialExpressionMultiply
   inputs [
    A {from=MaterialExpressionMultiply_54}
    B {from=""}
   ]
   properties {constB=2}
  }
  MaterialExpressionMultiply_6 {
   class=MaterialExpressionMultiply
   inputs [
    A {from=MaterialExpressionTextureSample_0, output=B}
    B {from=MaterialExpressionStep_6}
   ]
  }
  MaterialExpressionMultiply_7 {
   class=MaterialExpressionMultiply
   inputs [
    A {from=MaterialExpressionNamedRerouteUsage_41}
    B {from=""}
   ]
   properties {constB=2}
  }
  MaterialExpressionMultiply_8 {
   class=MaterialExpressionMultiply
   inputs [
    A {from=MaterialExpressionNamedRerouteDeclaration_32}
    B {from=MaterialExpressionSaturate_1}
   ]
  }
  MaterialExpressionMultiply_9 {
   class=MaterialExpressionMultiply
   inputs [
    A {from=MaterialExpressionMultiply_35}
    B {from=""}
   ]
   properties {constB=8}
  }
  MaterialExpressionNamedRerouteDeclaration_0 {
   class=MaterialExpressionNamedRerouteDeclaration
   inputs [None {from=MaterialExpressionLinearInterpolate_18}]
   properties {name=OUT_BaseColor, nodeColor=<rgba 0 0.8856 1 1>}
  }
  MaterialExpressionNamedRerouteDeclaration_11 {
   class=MaterialExpressionNamedRerouteDeclaration
   inputs [None {from=MaterialExpressionGetMaterialAttributes_0, output=EmissiveColor}]
   properties {name=IN_EmissiveColor, nodeColor=<rgba 0 0.8 1 1>}
  }
  MaterialExpressionNamedRerouteDeclaration_14 {
   class=MaterialExpressionNamedRerouteDeclaration
   inputs [None {from=MaterialExpressionPower_1}]
   properties {name=Amplitude, nodeColor=<rgba 0 0.5547 1 1>}
  }
  MaterialExpressionNamedRerouteDeclaration_19 {
   class=MaterialExpressionNamedRerouteDeclaration
   inputs [None {from=MaterialExpressionSaturate_0}]
   properties {name=Specturm, nodeColor=<rgba 0 0.5547 1 1>}
  }
  MaterialExpressionNamedRerouteDeclaration_20 {
   class=MaterialExpressionNamedRerouteDeclaration
   inputs [None {from=MaterialExpressionConvert_0}]
   properties {name=BassFreq, nodeColor=<rgba 0.0078 0 1 1>}
  }
  MaterialExpressionNamedRerouteDeclaration_21 {
   class=MaterialExpressionNamedRerouteDeclaration
   inputs [None {from=MaterialExpressionConvert_0}]
   properties {name=LowFreq, nodeColor=<rgba 0.0078 0 1 1>}
  }
  MaterialExpressionNamedRerouteDeclaration_25 {
   class=MaterialExpressionNamedRerouteDeclaration
   inputs [None {from=MaterialExpressionFrac_0}]
   properties {name=FakeAmp, nodeColor=<rgba 0.8047 0 1 1>}
  }
  MaterialExpressionNamedRerouteDeclaration_26 {
   class=MaterialExpressionNamedRerouteDeclaration
   inputs [None {from=MaterialExpressionConvert_0}]
   properties {name=MidFreq, nodeColor=<rgba 0.0078 0 1 1>}
  }
  MaterialExpressionNamedRerouteDeclaration_27 {
   class=MaterialExpressionNamedRerouteDeclaration
   inputs [None {from=MaterialExpressionConstant_1}]
   properties {name=FakeSpec, nodeColor=<rgba 0.8047 0 1 1>}
  }
  MaterialExpressionNamedRerouteDeclaration_28 {
   class=MaterialExpressionNamedRerouteDeclaration
   inputs [None {from=MaterialExpressionConvert_0}]
   properties {name=HighFreq, nodeColor=<rgba 0.0078 0 1 1>}
  }
  MaterialExpressionNamedRerouteDeclaration_29 {
   class=MaterialExpressionNamedRerouteDeclaration
   inputs [None {from=MaterialExpressionMultiply_16}]
   properties {name=GemAMP, nodeColor=<rgba 1 0 0.2344 1>}
  }
  MaterialExpressionNamedRerouteDeclaration_30 {
   class=MaterialExpressionNamedRerouteDeclaration
   inputs [None {from=MaterialExpressionStep_2}]
   properties {name=AudioTriggered, nodeColor=<rgba 1 0 0.1641 1>}
  }
  MaterialExpressionNamedRerouteDeclaration_31 {
   class=MaterialExpressionNamedRerouteDeclaration
   inputs [None {from=MaterialExpressionSmoothStep_1}]
   properties {name=AudioOnTransition, nodeColor=<rgba 1 0 0.1641 1>}
  }
  MaterialExpressionNamedRerouteDeclaration_32 {
   class=MaterialExpressionNamedRerouteDeclaration
   inputs [None {from=MaterialExpressionMultiply_6}]
   properties {name=GemMask, nodeColor=<rgba 1 0 0.2109 1>}
  }
  MaterialExpressionNamedRerouteDeclaration_33 {
   class=MaterialExpressionNamedRerouteDeclaration
   inputs [None {from=MaterialExpressionMultiply_9}]
   properties {name=GemSpec, desc=Not Used, nodeColor=<rgba 1 0 0.2344 1>}
  }
  MaterialExpressionNamedRerouteDeclaration_34 {
   class=MaterialExpressionNamedRerouteDeclaration
   inputs [None {from=MaterialExpressionMultiply_47}]
   properties {name=RimLighht, nodeColor=<rgba 0.9219 0 1 1>}
  }
  MaterialExpressionNamedRerouteDeclaration_5 {
   class=MaterialExpressionNamedRerouteDeclaration
   inputs [None {from=MaterialExpressionAdd_14}]
   properties {name=OUT_EmissiveColor, nodeColor=<rgba 0 0.8 1 1>}
  }
  MaterialExpressionNamedRerouteDeclaration_6 {
   class=MaterialExpressionNamedRerouteDeclaration
   inputs [None {from=MaterialExpressionGetMaterialAttributes_0, output=BaseColor}]
   properties {name=IN_BaseColor, nodeColor=<rgba 0 0.8 1 1>}
  }
  MaterialExpressionNamedRerouteUsage_0 {class=MaterialExpressionNamedRerouteUsage}
  MaterialExpressionNamedRerouteUsage_15 {class=MaterialExpressionNamedRerouteUsage}
  MaterialExpressionNamedRerouteUsage_16 {class=MaterialExpressionNamedRerouteUsage}
  MaterialExpressionNamedRerouteUsage_17 {class=MaterialExpressionNamedRerouteUsage}
  MaterialExpressionNamedRerouteUsage_18 {class=MaterialExpressionNamedRerouteUsage}
  MaterialExpressionNamedRerouteUsage_21 {class=MaterialExpressionNamedRerouteUsage}
  MaterialExpressionNamedRerouteUsage_22 {class=MaterialExpressionNamedRerouteUsage}
  MaterialExpressionNamedRerouteUsage_23 {class=MaterialExpressionNamedRerouteUsage}
  MaterialExpressionNamedRerouteUsage_24 {class=MaterialExpressionNamedRerouteUsage}
  MaterialExpressionNamedRerouteUsage_3 {class=MaterialExpressionNamedRerouteUsage}
  MaterialExpressionNamedRerouteUsage_35 {class=MaterialExpressionNamedRerouteUsage}
  MaterialExpressionNamedRerouteUsage_36 {class=MaterialExpressionNamedRerouteUsage}
  MaterialExpressionNamedRerouteUsage_38 {class=MaterialExpressionNamedRerouteUsage}
  MaterialExpressionNamedRerouteUsage_4 {class=MaterialExpressionNamedRerouteUsage}
  MaterialExpressionNamedRerouteUsage_41 {class=MaterialExpressionNamedRerouteUsage}
  MaterialExpressionNamedRerouteUsage_42 {class=MaterialExpressionNamedRerouteUsage}
  MaterialExpressionOneMinus_0 {class=MaterialExpressionOneMinus, inputs [None {from=MaterialExpressionMax_0}]}
  MaterialExpressionOneMinus_1 {class=MaterialExpressionOneMinus, inputs [None {from=MaterialExpressionFresnel_0}]}
  MaterialExpressionPower_0 {
   class=MaterialExpressionPower
   inputs [
    Base {from=MaterialExpressionLinearInterpolate_6}
    Exp {from=MaterialExpressionScalarParameter_2}
   ]
   properties {constExponent=5}
  }
  MaterialExpressionPower_1 {
   class=MaterialExpressionPower
   inputs [
    Base {from=MaterialExpressionLinearInterpolate_7}
    Exp {from=MaterialExpressionScalarParameter_1}
   ]
   properties {constExponent=3}
  }
  MaterialExpressionPower_3 {
   class=MaterialExpressionPower
   inputs [
    Base {from=MaterialExpressionFresnel_2}
    Exp {from=""}
   ]
   properties {constExponent=4}
  }
  MaterialExpressionPower_5 {
   class=MaterialExpressionPower
   inputs [
    Base {from=MaterialExpressionNamedRerouteUsage_15}
    Exp {from=""}
   ]
   properties {constExponent=3}
  }
  MaterialExpressionPower_7 {
   class=MaterialExpressionPower
   inputs [
    Base {from=MaterialExpressionOneMinus_1}
    Exp {from=""}
   ]
   properties {constExponent=3}
  }
  MaterialExpressionPower_8 {
   class=MaterialExpressionPower
   inputs [
    Base {from=MaterialExpressionFresnel_1}
    Exp {from=""}
   ]
   properties {constExponent=3}
  }
  MaterialExpressionSaturate_0 {class=MaterialExpressionSaturate, inputs [None {from=MaterialExpressionAdd_0}]}
  MaterialExpressionSaturate_1 {class=MaterialExpressionSaturate, inputs [None {from=MaterialExpressionOneMinus_0}]}
  MaterialExpressionScalarParameter_0 {class=MaterialExpressionScalarParameter, properties {defaultValue=1, group=Audio Reactivity, parameterName=EnableAudioReactivity}}
  MaterialExpressionScalarParameter_1 {class=MaterialExpressionScalarParameter, properties {defaultValue=8, group=Audio Reactivity, parameterName=Audio_Amplitude_Exp}}
  MaterialExpressionScalarParameter_2 {class=MaterialExpressionScalarParameter, properties {defaultValue=3, group=Audio Reactivity, parameterName=Audio_Spectrum_Exp}}
  MaterialExpressionScalarParameter_4 {class=MaterialExpressionScalarParameter, properties {group=Audio Reactivity, parameterName=EnableFrontendFakeAudioLoop}}
  MaterialExpressionScalarParameter_7 {class=MaterialExpressionScalarParameter, properties {defaultValue=0.75, group=Audio Reactivity, parameterName=FakeAudio_IdleSpeed}}
  MaterialExpressionSetMaterialAttributes_0 {
   class=MaterialExpressionSetMaterialAttributes
   inputs [
    MaterialAttributes {from=MaterialExpressionGetMaterialAttributes_0, output=MaterialAttributes}
    BaseColor {from=MaterialExpressionNamedRerouteUsage_0, output=""}
    EmissiveColor {from=MaterialExpressionNamedRerouteUsage_4, output=""}
    Specular {from=MaterialExpressionLinearInterpolate_3, output=""}
   ]
   properties {attributeSetTypes [#1, #2, #3]}
  }
  MaterialExpressionShaderStageSwitch_0 {
   class=MaterialExpressionShaderStageSwitch
   inputs [
    PixelShader {from=MaterialExpressionVertexInterpolator_0, output=PS}
    VertexShader {from=MaterialExpressionMax_7, output=""}
   ]
   properties {}
  }
  MaterialExpressionSine_0 {class=MaterialExpressionSine, inputs [None {from=MaterialExpressionAdd_3}]}
  MaterialExpressionSine_2 {class=MaterialExpressionSine, inputs [None {from=MaterialExpressionAdd_5}]}
  MaterialExpressionSmoothStep_1 {
   class=MaterialExpressionSmoothStep
   inputs [
    Min {from="", output=""}
    Max {from="", output=""}
    Value {from=MaterialExpressionMaterialFunctionCall_1, output=Amplitude}
   ]
   properties {constMax=0.9, constMin=0.1}
  }
  MaterialExpressionSphereMask_0 {
   class=MaterialExpressionSphereMask
   inputs [
    A {from=MaterialExpressionLocalPosition_0, output=XYZ}
    B {from=MaterialExpressionLinearInterpolate_26, output=""}
    Radius {from=MaterialExpressionLinearInterpolate_0, output=""}
    Hardness {from="", output=""}
   ]
   properties {attenuationRadius=30, hardnessPercent=0}
  }
  MaterialExpressionSphereMask_1 {
   class=MaterialExpressionSphereMask
   inputs [
    A {from=MaterialExpressionLocalPosition_0, output=XYZ}
    B {from=MaterialExpressionLinearInterpolate_27, output=""}
    Radius {from=MaterialExpressionLinearInterpolate_0, output=""}
    Hardness {from="", output=""}
   ]
   properties {attenuationRadius=30, hardnessPercent=0}
  }
  MaterialExpressionSphereMask_2 {
   class=MaterialExpressionSphereMask
   inputs [
    A {from=MaterialExpressionLocalPosition_0, output=XYZ}
    B {from=MaterialExpressionConstant3Vector_4, output=""}
    Radius {from=MaterialExpressionLinearInterpolate_1, output=""}
    Hardness {from="", output=""}
   ]
   properties {attenuationRadius=15, hardnessPercent=0}
  }
  MaterialExpressionSphereMask_3 {
   class=MaterialExpressionSphereMask
   inputs [
    A {from=MaterialExpressionLocalPosition_0, output=XYZ}
    B {from=MaterialExpressionLinearInterpolate_28, output=""}
    Radius {from=MaterialExpressionLinearInterpolate_29, output=""}
    Hardness {from="", output=""}
   ]
   properties {attenuationRadius=6, hardnessPercent=0}
  }
  MaterialExpressionStep_10 {
   class=MaterialExpressionStep
   inputs [
    Y {from=MaterialExpressionComponentMask_6, output=""}
    X {from="", output=""}
   ]
   properties {constX=0.5}
  }
  MaterialExpressionStep_2 {
   class=MaterialExpressionStep
   inputs [
    Y {from="", output=""}
    X {from=MaterialExpressionLinearInterpolate_7, output=""}
   ]
   properties {constX=0.01, constY=0.1}
  }
  MaterialExpressionStep_6 {
   class=MaterialExpressionStep
   inputs [
    Y {from="", output=""}
    X {from=MaterialExpressionTextureSample_0, output=G}
   ]
   properties {constX=0.002, constY=0.002}
  }
  MaterialExpressionStep_7 {
   class=MaterialExpressionStep
   inputs [
    Y {from=MaterialExpressionTextureSample_0, output=R}
    X {from=MaterialExpressionNamedRerouteUsage_16, output=""}
   ]
   properties {}
  }
  MaterialExpressionStep_8 {
   class=MaterialExpressionStep
   inputs [
    Y {from="", output=""}
    X {from=MaterialExpressionTextureSample_0, output=B}
   ]
   properties {constX=0.002, constY=0.6}
  }
  MaterialExpressionStep_9 {
   class=MaterialExpressionStep
   inputs [
    Y {from=MaterialExpressionComponentMask_5, output=""}
    X {from="", output=""}
   ]
   properties {constX=0.5}
  }
  MaterialExpressionSubtract_1 {
   class=MaterialExpressionSubtract
   inputs [
    A {from=MaterialExpressionMultiply_54}
    B {from=""}
   ]
  }
  MaterialExpressionTextureSample_0 {
   class=MaterialExpressionTextureSample
   inputs [
    UVs {from=""}
    Tex {from=""}
    Apply View MipBias {from=""}
   ]
   properties {
    samplerSource=SSM_Wrap_WorldGroupSettings
    texture=/BRCosmetics/Characters/Player/Female/Medium/Bodies/F_MED_SprintLuckSlaw/Textures/T_SprintLuckSlaw_TechArt
   }
  }
  MaterialExpressionTime_1 {class=MaterialExpressionTime}
  MaterialExpressionTime_2 {class=MaterialExpressionTime}
  MaterialExpressionVectorParameter_0 {
   class=MaterialExpressionVectorParameter
   properties {group=Audio Reactivity, parameterName=AudioReactiveEmissveColor, defaultValue=<rgba 0.25 0.8452 1.5 1>}
  }
  MaterialExpressionVertexInterpolator_0 {class=MaterialExpressionVertexInterpolator, inputs [VS {from=MaterialExpressionMax_7, output=""}], properties {}}
 }}
