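# Post-process style body AnimBP for the SprintLuckSlaw female character that copies the parent mesh pose, adding rig corrections and secondary motion.
# A Control Rig driven by smoothed skydiving state and foot scale tweaks feeds stacked angular and linear AnimDynamics on upper-torso bones, bypassed unless bPbBlendEnabled.
# DefaultSlot overrides skirt bones, weighted by the front-end idle curve.
# The event graph updates these values and flags gadget boots, unused here.
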AnimBlueprint /BRCosmetics/Characters/Player/Female/Medium/Bodies/F_MED_SprintLuckSlaw/Meshes/F_MED_SprintLuckSlaw_AnimBP
target_skeleton=&/F_MED_SprintLuckSlaw_Skeleton
graphs [
 AnimGraph {nodes [
   Output Pose(Result=Layered blend per bone(
     BasePose=Use cached pose 'StartPose'()
     BlendPoses_0=Slot 'DefaultSlot'()
     BlendWeights_0=Get Custom Idle Is Playing Float() -> (Custom Idle Is Playing Float:double)
    ) {layerSetup [{branchFilters [{boneName=dyn_skirt_b_m}, {boneName=dyn_skirt_f_m}]}]}) {layerGroup=None}
   Save cached pose 'StartPose'(Pose=Blend Poses by bool(
     BlendPose_0=Use cached pose 'my_pose'()
     BlendPose_1=Use cached pose 'my_pose_1'()
     bActiveValue=Get bPbBlendEnabled() -> (bPbBlendEnabled:bool)
    )) {cacheName=StartPose}
   n1 = "Transform (Modify) Bone - Bone: dyn_upperTorso_l"(ComponentPose=Local To Component(LocalPose=Use cached pose 'my_pose_1'()), Alpha=1, Translation="0,0,1", Rotation="0,0,0", Scale="1,1,1") {translationMode=BMM_Additive, translationSpace=BCS_BoneSpace, boneToModify {boneName=dyn_upperTorso_l}, translation {z=1}}
   n2 = "Transform (Modify) Bone - Bone: dyn_upperTorso_r"(ComponentPose=@n1, Alpha=1, Translation="0,0,1", Rotation="0,0,0", Scale="1,1,1") {translationMode=BMM_Additive, translationSpace=BCS_BoneSpace, boneToModify {boneName=dyn_upperTorso_r}, translation {z=1}}
   n3 = Get Sim Space Settings() -> (Sim Space Settings:Anim Phys Sim Space Settings Structure)
   n4 = Anim Dynamics - Bone: dyn_upperTorso_r!(ComponentPose=@n2, Alpha=1, AngularSpringConstant:float=5000, SimSpaceSettings:Anim Phys Sim Space Settings Structure=@n3) -> (Pose) {
    angularBiasOverride=1
    angularDampingOverride=0.9999
    angularSpringConstant=5000
    bAngularSpring=true
    bGravityOverrideInSimSpace=true
    bOverrideAngularBias=true
    bOverrideAngularDamping=true
    bUseGravityOverride=true
    linearDampingOverride=0.9999
    linearSpringConstant=100
    simulationSpace=RootRelative
    boundBone {boneName=dyn_upperTorso_r}
    componentAppliedLinearAccClamp=<2 2 2>
    gravityOverride {z=-980}
    physicsBodyDefinitions [{
      boundBone {boneName=dyn_upperTorso_r}
      boxExtents=<21 21 21>
      constraintSetup {angularLimitsMax=<0 1 5>, angularLimitsMin=<0 -5 -5>, angularTarget=<1 0 0>}
      localJointOffset=<2 0 0>
     }]
    relativeSpaceBone {boneName=spine_04}
    simSpaceSettings {maxAngularAcceleration=45, maxAngularVelocity=8, simSpaceAngularAlpha=0.25}
   }
   n5 = Anim Dynamics - Bone: dyn_upperTorso_l!(ComponentPose=@n4, Alpha=1, AngularSpringConstant:float=5000, SimSpaceSettings:Anim Phys Sim Space Settings Structure=@n3) -> (Pose) {
    angularBiasOverride=1
    angularDampingOverride=0.9999
    angularSpringConstant=5000
    bAngularSpring=true
    bGravityOverrideInSimSpace=true
    bOverrideAngularBias=true
    bOverrideAngularDamping=true
    bUseGravityOverride=true
    linearDampingOverride=0.9999
    linearSpringConstant=100
    simulationSpace=RootRelative
    boundBone {boneName=dyn_upperTorso_l}
    componentAppliedLinearAccClamp=<2 2 2>
    gravityOverride {z=-980}
    physicsBodyDefinitions [{
      boundBone {boneName=dyn_upperTorso_l}
      boxExtents=<20 20 20>
      constraintSetup {angularLimitsMax=<0 1 5>, angularLimitsMin=<0 -5 -5>, angularTarget=<1 0 0>}
      localJointOffset=<2 0 0>
     }]
    relativeSpaceBone {boneName=spine_04}
    simSpaceSettings {maxAngularAcceleration=45, maxAngularVelocity=8, simSpaceAngularAlpha=0.25}
   }
   n6 = Anim Dynamics - Bone: dyn_upperTorso_r!(ComponentPose=@n5, Alpha=1, GravityOverride:Vector="0,0,-200", LinearSpringConstant:float=150) -> (Pose) {
    angularBiasOverride=1
    angularDampingOverride=0.99
    angularSpringConstant=10000
    bGravityOverrideInSimSpace=true
    bLinearSpring=true
    bOverrideAngularBias=true
    bOverrideLinearDamping=true
    bUseGravityOverride=true
    linearDampingOverride=0.9999
    linearSpringConstant=180
    simulationSpace=RootRelative
    boundBone {boneName=dyn_upperTorso_r}
    componentAppliedLinearAccClamp=<5 5 5>
    gravityOverride {z=-200}
    physicsBodyDefinitions [{
      boundBone {boneName=dyn_upperTorso_r}
      boxExtents=<80 80 80>
      constraintSetup {angularLimitsMax=<0 0 0>, angularLimitsMin=<0 0 0>, angularTarget=<1 0 0>, linearAxesMax=<0 0.6 1.15>, linearAxesMin=<0 -0.6 -0.6>}
      localJointOffset=<0.2 0 0>
     }]
    relativeSpaceBone {boneName=spine_04}
    simSpaceSettings {maxAngularAcceleration=50, maxAngularVelocity=10, simSpaceAngularAlpha=0.1}
   }
   n7 = Anim Dynamics - Bone: dyn_upperTorso_l!(ComponentPose=@n6, Alpha=1, GravityOverride:Vector="0,0,-200", LinearSpringConstant:float=150) -> (Pose) {
    angularBiasOverride=1
    angularDampingOverride=0.99
    angularSpringConstant=10000
    bGravityOverrideInSimSpace=true
    bLinearSpring=true
    bOverrideAngularBias=true
    bOverrideLinearDamping=true
    bUseGravityOverride=true
    linearDampingOverride=0.9999
    linearSpringConstant=180
    simulationSpace=RootRelative
    boundBone {boneName=dyn_upperTorso_l}
    componentAppliedLinearAccClamp=<5 5 5>
    gravityOverride {z=-200}
    physicsBodyDefinitions [{
      boundBone {boneName=dyn_upperTorso_l}
      boxExtents=<90 90 90>
      constraintSetup {angularLimitsMax=<0 0 0>, angularLimitsMin=<0 0 0>, angularTarget=<1 0 0>, linearAxesMax=<0 0.6 1.15>, linearAxesMin=<0 -0.6 -0.6>}
      localJointOffset=<0.2 0 0>
     }]
    relativeSpaceBone {boneName=spine_04}
    simSpaceSettings {maxAngularAcceleration=50, maxAngularVelocity=10, simSpaceAngularAlpha=0.1}
   }
   Save cached pose 'my_pose'(Pose=Component To Local(ComponentPose=@n7)) {cacheName=my_pose}
   n8 = Control Rig!(Source=Copy Pose From Mesh() {bUseAttachedParent=true}, Alpha, Skydiving:double=Get Skydiving() -> (Skydiving:double)) -> (Pose) {controlRigAssetReference {controlRigAsset=&/F_MED_SprintLuckSlaw_CtrlRig}}
   n9 = "Transform (Modify) Bone - Bone: foot_reparent_l"!(ComponentPose=Local To Component(LocalPose=@n8), bAlphaBoolEnabled:bool=True, Translation="0,0,0", Rotation="0,0,0", Scale="0.1,0.1,0.1") -> (Pose) {
    alphaInputType=Bool
    scaleMode=BMM_Additive
    alphaBoolBlend {blendOption=Linear}
    boneToModify {boneName=foot_reparent_l}
    scale=<0.1 0.1 0.1>
   }
   n10 = "Transform (Modify) Bone - Bone: foot_reparent_r"!(ComponentPose=@n9, bAlphaBoolEnabled:bool=True, Translation="0,0,0", Rotation="0,0,0", Scale="0.1,0.1,0.1") -> (Pose) {
    alphaInputType=Bool
    scaleMode=BMM_Additive
    alphaBoolBlend {blendOption=Linear}
    boneToModify {boneName=foot_reparent_r}
    scale=<0.1 0.1 0.1>
   }
   Save cached pose 'my_pose_1'(Pose=Component To Local(ComponentPose=@n10)) {cacheName=my_pose_1}
  ]}
 EventGraph {nodes [
   Event BlueprintUpdateAnimation()
   > n1 = Sequence!(execute) -> (then_0, then_1, then_2:Exec)
   Branch(execute=@n1.then_2, Condition=Get bIsInFrontEnd())
   > Set Custom Idle Is Playing Float(Custom Idle Is Playing Float:double=GetCurveValue(self=GetOwnerFrontEndAnimInstance(), CurveName=Frontend_Idle_Bypass)) -> (Output_Get:double)
   n2 = Branch(execute=@n1.then_1, Condition=Get bIsSkyDiving())
   n3 = float / float(A=Get Delta Time X(), B=2)
   > Set Skydiving(Skydiving:double=FInterpTo(Current=Get Skydiving(), Target=1, DeltaTime=@n3, InterpSpeed=20)) -> (Output_Get:double)
   Set Skydiving(execute=@n2.else, Skydiving=FInterpTo(Current=Get Skydiving(), Target=0, DeltaTime=@n3, InterpSpeed=20))
   n4 = Get As Fort Player Pawn Athena() -> (As Fort Player Pawn Athena:Fort Player Pawn Athena Object Reference)
   Is Valid(exec=@n1.then_0, InputObject=@n4)
   n5 = GetSkeletalMeshForPartType(self=@n4, PartType=ExtraPart)
   > Set CustomBootsEquip(CustomBootsEquip:bool=Contains(SearchIn=GetDisplayName(Object=@n5), Substring=BootsTemplate_Gadget)) -> (Output_Get:bool)
   Event OnCharacterCustomizationLoaded()
   n6 = TryGetPawnOwner()
   > Is Valid(InputObject=@n6)
   > n7 = Cast To FortPlayerPawnAthena(Object=@n6)
   > Set As Fort Player Pawn Athena(As Fort Player Pawn Athena:Fort Player Pawn Athena Object Reference=@n7) -> (Output_Get:Fort Player Pawn Athena Object Reference)
  ]}
]
variables [
 As Fort Player Pawn Athena {type=FortPlayerPawnAthena}
 Custom Idle Is Playing Float {type=number}
 CustomBootsEquip {type=boolean}
 Delta Time X {type=number}
 Sim Space Settings {
  type=AnimPhysSimSpaceSettings
  default {maxAngularAcceleration=45, maxAngularVelocity=8, simSpaceAngularAlpha=0.25, externalAngularVelocity=<0 0 0>}
 }
 Skydiving {type=number}
]
