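# Post-process Animation Blueprint for the F_MED_SprintLuckSlaw female medium player body cosmetic skeleton, correcting foot helper bone rotations.
# The input pose passes through four chained rotation-only Constraint nodes in component space at full weight.
# These make the left and right foot_reparent and ball_reparent bones copy the rotation of the matching foot and ball bones.
# The event graph holds only unconnected default nodes, and there are no variables.
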AnimBlueprint /BRCosmetics/Characters/Player/Female/Medium/Bodies/F_MED_SprintLuckSlaw/Meshes/F_MED_SprintLuckSlaw_Skeleton_AnimBlueprint_Post_process
target_skeleton=&/F_MED_SprintLuckSlaw_Skeleton
graphs [
 AnimGraph {nodes [
   n1 = "Constraint - foot_reparent_l (R)"(ComponentPose=Local To Component(LocalPose=Input Pose()), Alpha=1, ConstraintWeights_0=1) {boneToModify {boneName=foot_reparent_l}, constraintSetup [{transformType=Rotation, targetBone {boneName=foot_l}}], constraintWeights [1]}
   n2 = "Constraint - ball_reparent_l (R)"(ComponentPose=@n1, Alpha=1, ConstraintWeights_0=1) {boneToModify {boneName=ball_reparent_l}, constraintSetup [{transformType=Rotation, targetBone {boneName=ball_l}}], constraintWeights [1]}
   n3 = "Constraint - foot_reparent_r (R)"(ComponentPose=@n2, Alpha=1, ConstraintWeights_0=1) {boneToModify {boneName=foot_reparent_r}, constraintSetup [{transformType=Rotation, targetBone {boneName=foot_r}}], constraintWeights [1]}
   n4 = "Constraint - ball_reparent_r (R)"(ComponentPose=@n3, Alpha=1, ConstraintWeights_0=1) {boneToModify {boneName=ball_reparent_r}, constraintSetup [{transformType=Rotation, targetBone {boneName=ball_r}}], constraintWeights [1]}
   Output Pose(Result=Component To Local(ComponentPose=@n4))
  ]}
 EventGraph {nodes [
   TryGetPawnOwner()
   Event BlueprintUpdateAnimation()
  ]}
]
variables []
