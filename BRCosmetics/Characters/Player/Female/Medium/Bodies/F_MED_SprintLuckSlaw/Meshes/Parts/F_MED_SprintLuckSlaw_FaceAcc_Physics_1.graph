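# Physics asset for the face accessory mesh of the SprintLuckSlaw female medium player cosmetic, simulating two dangling hair strands.
# Left and right five-bone dyn_hair capsule chains hang from base_hair bones through tightly limited swing joints with zero twist.
# Capsules on the head, neck, upper spine, clavicles and upper arms provide body collision shapes.
# Each chain tip also has a fully free angular constraint to neck_01.
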PhysicsAsset /BRCosmetics/Characters/Player/Female/Medium/Bodies/F_MED_SprintLuckSlaw/Meshes/Parts/F_MED_SprintLuckSlaw_FaceAcc_Physics_1
bodies {
 base_hair_l [{shapeType=Capsule, length=0.4, radius=1, shapeName=base_hair_l_capsule, rotation=<rot 0 -90 0>}]
 base_hair_r [{shapeType=Capsule, length=0.4, radius=1, shapeName=base_hair_r_capsule, rotation=<rot 0 -90 0>}]
 clavicle_l [{shapeType=Capsule, length=7.5, radius=5.8, shapeName=clavicle_l_capsule, center=<4.0176 0.3067 -0.2>, rotation=<rot 90 0 0>}]
 clavicle_r [{shapeType=Capsule, length=7.5, radius=5.8, shapeName=clavicle_r_capsule, center=<-4.0176 -0.3067 0.2>, rotation=<rot 90 180 0>}]
 dyn_hair_01_l [{shapeType=Capsule, length=4, radius=2.8, shapeName=dyn_hair_01_l_capsule, center=<3 0 0>, rotation=<rot 90 0 0>}]
 dyn_hair_01_r [{shapeType=Capsule, length=4, radius=2.8, shapeName=dyn_hair_01_r_capsule, center=<3 0 0>, rotation=<rot 90 0 0>}]
 dyn_hair_02_l [{shapeType=Capsule, length=4, radius=2.8, shapeName=dyn_hair_02_l_capsule, center=<3 0 0>, rotation=<rot 90 0 0>}]
 dyn_hair_02_r [{shapeType=Capsule, length=4, radius=2.8, shapeName=dyn_hair_02_r_capsule, center=<3 0 0>, rotation=<rot 90 0 0>}]
 dyn_hair_03_l [{shapeType=Capsule, length=4, radius=2.8, shapeName=dyn_hair_03_l_capsule, center=<3 0 0>, rotation=<rot 90 0 0>}]
 dyn_hair_03_r [{shapeType=Capsule, length=4, radius=2.8, shapeName=dyn_hair_03_r_capsule, center=<3 0 0>, rotation=<rot 90 0 0>}]
 dyn_hair_04_l [{shapeType=Capsule, length=4, radius=2.8, shapeName=dyn_hair_04_l_capsule, center=<3 0 0>, rotation=<rot 90 0 0>}]
 dyn_hair_04_r [{shapeType=Capsule, length=4, radius=2.8, shapeName=dyn_hair_04_r_capsule, center=<3 0 0>, rotation=<rot 90 0 0>}]
 dyn_hair_05_l [{shapeType=Capsule, length=2, radius=2.8, shapeName=dyn_hair_05_l_capsule, center=<2 0 0>, rotation=<rot 90 0 0>}]
 dyn_hair_05_r [{shapeType=Capsule, length=3, radius=2.8, shapeName=dyn_hair_05_r_capsule, center=<2 0 0>, rotation=<rot 90 0 0>}]
 head [
  {shapeType=Capsule, length=14, radius=15, shapeName=head_capsule, center=<-3.4243 17.7743 -12.724>, rotation=<rot 0 0 -90>}
  {shapeType=Capsule, length=14, radius=15, shapeName=head_capsule, center=<-3.4243 17 12.724>, rotation=<rot 0 0 90>}
  {shapeType=Capsule, length=17, radius=13, shapeName=head_capsule, center=<-12.2751 11.1301 0>}
 ]
 neck_01 [{shapeType=Capsule, length=4, radius=6, shapeName=neck_01_capsule, center=<0 1.4815 0>}]
 neck_02 [{shapeType=Capsule, length=3, radius=5, shapeName=neck_02_capsule}]
 spine_04 [{shapeType=Capsule, length=21.5, radius=7.3, shapeName=spine_04_capsule}]
 spine_05 [{shapeType=Capsule, length=19, radius=7.3, shapeName=spine_05_capsule}]
 upperarm_l [{shapeType=Capsule, length=15, radius=5, shapeName=upperarm_l_capsule, center=<14.3291 0 0>, rotation=<rot 90 180 0>}]
 upperarm_r [{shapeType=Capsule, length=15, radius=5, shapeName=upperarm_r_capsule, center=<-14.3291 0 0>, rotation=<rot 90 0 0>}]
}
constraints [
 {bone1Name=dyn_hair_01_l, bone2Name=base_hair_l, swing1LimitDegrees=15, swing2LimitDegrees=5, twistLimitDegrees=0}
 {bone1Name=dyn_hair_02_l, bone2Name=dyn_hair_01_l, swing1LimitDegrees=15, swing2LimitDegrees=5, twistLimitDegrees=0}
 {bone1Name=dyn_hair_03_l, bone2Name=dyn_hair_02_l, swing1LimitDegrees=15, swing2LimitDegrees=5, twistLimitDegrees=0}
 {bone1Name=dyn_hair_04_l, bone2Name=dyn_hair_03_l, swing1LimitDegrees=15, swing2LimitDegrees=5, twistLimitDegrees=0}
 {bone1Name=dyn_hair_05_l, bone2Name=dyn_hair_04_l, swing1LimitDegrees=15, swing2LimitDegrees=5, twistLimitDegrees=0}
 {bone1Name=dyn_hair_01_r, bone2Name=base_hair_r, swing1LimitDegrees=15, swing2LimitDegrees=5, twistLimitDegrees=0}
 {bone1Name=dyn_hair_02_r, bone2Name=dyn_hair_01_r, swing1LimitDegrees=15, swing2LimitDegrees=5, twistLimitDegrees=0}
 {bone1Name=dyn_hair_03_r, bone2Name=dyn_hair_02_r, swing1LimitDegrees=15, swing2LimitDegrees=5, twistLimitDegrees=0}
 {bone1Name=dyn_hair_04_r, bone2Name=dyn_hair_03_r, swing1LimitDegrees=15, swing2LimitDegrees=5, twistLimitDegrees=0}
 {bone1Name=dyn_hair_05_r, bone2Name=dyn_hair_04_r, swing1LimitDegrees=15, swing2LimitDegrees=5, twistLimitDegrees=0}
 {
  bone1Name=dyn_hair_05_r
  bone2Name=neck_01
  swing1LimitDegrees=45
  swing1Motion=Free
  swing2LimitDegrees=45
  swing2Motion=Free
  twistLimitDegrees=45
  twistMotion=Free
 }
 {
  bone1Name=dyn_hair_05_l
  bone2Name=neck_01
  swing1LimitDegrees=45
  swing1Motion=Free
  swing2LimitDegrees=45
  swing2Motion=Free
  twistLimitDegrees=45
  twistMotion=Free
 }
]
