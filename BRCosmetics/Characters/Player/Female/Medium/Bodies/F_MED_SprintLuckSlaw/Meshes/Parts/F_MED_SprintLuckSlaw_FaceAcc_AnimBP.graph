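# Animation Blueprint for the SprintLuckSlaw face accessory cosmetic, adding secondary motion to its dynamic hair bones.
# Poses copied from the attached parent and Body meshes blend per bone, pass through a Control Rig fed backpack and RBAN flags, then get cached.
# The cache feeds AnimDynamics chains and RigidBody nodes, selected by nested bool blends with separate front-end and in-level branches.
# Events set backbling state, sim-space settings, gravity and rigid-body parameters.
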AnimBlueprint /BRCosmetics/Characters/Player/Female/Medium/Bodies/F_MED_SprintLuckSlaw/Meshes/Parts/F_MED_SprintLuckSlaw_FaceAcc_AnimBP
target_skeleton=&/F_MED_SprintLuckSlaw_FaceAcc_Skeleton
graphs [
 AnimGraph {nodes [
   n1 = Anim Dynamics - Bone: dyn_hair_02_l!(
    ComponentPose=Local To Component(LocalPose=Use cached pose 'Main Cache'())
    Alpha=1
    GravityOverride:Vector="0,0,0"
    SimSpaceSettings:Anim Phys Sim Space Settings Structure=(SimSpaceAngularAlpha=0,MaxAngularVelocity=10000,MaxAngularAcceleration=10000,ExternalAngularVelocity=(X=0,Y=0,Z=0))
    ExternalForce="0,0,0"
   ) -> (Pose) {
    angularDampingOverride=0.999
    angularSpringConstant=4000
    bAngularSpring=true
    bGravityOverrideInSimSpace=true
    bOverrideAngularDamping=true
    bUseGravityOverride=true
    simulationSpace=RootRelative
    boundBone {boneName=dyn_hair_02_l}
    componentAppliedLinearAccClamp=<1000 1000 1000>
    componentLinearAccScale=<0.5 0.5 0.5>
    componentLinearVelScale=<0.3 0.3 0.3>
    physicsBodyDefinitions [{
      boundBone {boneName=dyn_hair_02_l}
      boxExtents=<20 20 20>
      constraintSetup {angularLimitsMax=<0 5 0>, angularLimitsMin=<0 -5 0>, angularTarget=<1 0 0>}
      localJointOffset=<3 0 0>
     }]
   }
   n2 = Anim Dynamics - Bone: dyn_hair_03_l!(
    ComponentPose=@n1
    Alpha=1
    GravityOverride:Vector="0,0,0"
    SimSpaceSettings:Anim Phys Sim Space Settings Structure=(SimSpaceAngularAlpha=0,MaxAngularVelocity=10000,MaxAngularAcceleration=10000,ExternalAngularVelocity=(X=0,Y=0,Z=0))
    ExternalForce="0,0,0"
   ) -> (Pose) {
    angularDampingOverride=0.999
    angularSpringConstant=4000
    bAngularSpring=true
    bOverrideAngularDamping=true
    bUseGravityOverride=true
    simulationSpace=RootRelative
    boundBone {boneName=dyn_hair_03_l}
    componentAppliedLinearAccClamp=<1000 1000 1000>
    componentLinearAccScale=<0.5 0.5 0.5>
    componentLinearVelScale=<0.3 0.3 0.3>
    physicsBodyDefinitions [{
      boundBone {boneName=dyn_hair_03_l}
      boxExtents=<20 20 20>
      constraintSetup {angularLimitsMax=<0 5 0>, angularLimitsMin=<0 -5 0>, angularTarget=<1 0 0>}
      localJointOffset=<3 0 0>
     }]
   }
   n3 = Anim Dynamics - Bone: dyn_hair_02_r!(
    ComponentPose=@n2
    Alpha=1
    GravityOverride:Vector="0,0,0"
    SimSpaceSettings:Anim Phys Sim Space Settings Structure=(SimSpaceAngularAlpha=0,MaxAngularVelocity=10000,MaxAngularAcceleration=10000,ExternalAngularVelocity=(X=0,Y=0,Z=0))
    ExternalForce="0,0,0"
   ) -> (Pose) {
    angularDampingOverride=0.999
    angularSpringConstant=4000
    bAngularSpring=true
    bOverrideAngularDamping=true
    bUseGravityOverride=true
    simulationSpace=RootRelative
    boundBone {boneName=dyn_hair_02_r}
    componentAppliedLinearAccClamp=<1000 1000 1000>
    componentLinearAccScale=<0.5 0.5 0.5>
    componentLinearVelScale=<0.3 0.3 0.3>
    physicsBodyDefinitions [{
      boundBone {boneName=dyn_hair_02_r}
      boxExtents=<20 20 20>
      constraintSetup {angularLimitsMax=<0 5 0>, angularLimitsMin=<0 0 0>, angularTarget=<1 0 0>}
      localJointOffset=<3 0 0>
     }]
   }
   n4 = Anim Dynamics - Bone: dyn_hair_03_r!(
    ComponentPose=@n3
    Alpha=1
    GravityOverride:Vector="0,0,0"
    SimSpaceSettings:Anim Phys Sim Space Settings Structure=(SimSpaceAngularAlpha=0,MaxAngularVelocity=10000,MaxAngularAcceleration=10000,ExternalAngularVelocity=(X=0,Y=0,Z=0))
    ExternalForce="0,0,0"
   ) -> (Pose) {
    angularDampingOverride=0.999
    angularSpringConstant=4000
    bAngularSpring=true
    bOverrideAngularDamping=true
    bUseGravityOverride=true
    simulationSpace=RootRelative
    boundBone {boneName=dyn_hair_03_r}
    componentAppliedLinearAccClamp=<1000 1000 1000>
    componentLinearAccScale=<0.5 0.5 0.5>
    componentLinearVelScale=<0.3 0.3 0.3>
    physicsBodyDefinitions [{
      boundBone {boneName=dyn_hair_03_r}
      boxExtents=<20 20 20>
      constraintSetup {angularLimitsMax=<0 5 0>, angularLimitsMin=<0 0 0>, angularTarget=<1 0 0>}
      localJointOffset=<3 0 0>
     }]
   }
   n5 = RigidBody!(
    ComponentPose=Local To Component(LocalPose=Use cached pose 'Main Cache'())
    Alpha
    OverrideWorldGravity:Vector="0,0,-980"
    ExternalForce
    SimSpaceSettings:Sim Space Settings Structure=(WorldAlpha=1,VelocityScaleZ=1,DampingAlpha=0,MaxLinearVelocity=400,MaxAngularVelocity=3,MaxLinearAcceleration=1400,MaxAngularAcceleration=12,ExternalLinearDragV=(X=0,Y=0,Z=0),ExternalLinearVelocity=(X=0,Y=0,Z=0),ExternalAngularVelocity=(X=0,Y=0,Z=0))
   ) -> (Pose) {
    bOverrideWorldGravity=true
    overridePhysicsAsset=&/F_MED_SprintLuckSlaw_FaceAcc_Physics_1
    simulationSpace=BaseBoneSpace
    baseBoneRef {boneName=spine_05}
    overrideWorldGravity {z=-980}
    simSpaceSettings {dampingAlpha=0, maxAngularAcceleration=12, maxAngularVelocity=3, maxLinearAcceleration=1400, maxLinearVelocity=400, worldAlpha=1}
   }
   n6 = RigidBody!(ComponentPose=Local To Component(LocalPose=Use cached pose 'Main Cache'()), Alpha, OverrideWorldGravity="0,0,-980", ExternalForce) -> (Pose) {
    bOverrideWorldGravity=true
    overridePhysicsAsset=&F_MED_SprintLuckSlaw_FaceAcc_Physics_1
    simulationSpace=BaseBoneSpace
    baseBoneRef {boneName=head}
    overrideWorldGravity {z=-980}
    simSpaceSettings {dampingAlpha=0, maxAngularAcceleration=12, maxAngularVelocity=3, maxLinearAcceleration=1400, maxLinearVelocity=400, worldAlpha=1}
   }
   n7 = Anim Dynamics - Bone: dyn_hair_02_l!(ComponentPose=Local To Component(LocalPose=Use cached pose 'Main Cache'()), Alpha=1, GravityOverride="0,0,0", ExternalForce="0,0,0") -> (Pose) {
    angularDampingOverride=0.999
    angularSpringConstant=4000
    bAngularSpring=true
    bGravityOverrideInSimSpace=true
    bOverrideAngularDamping=true
    bUseGravityOverride=true
    simulationSpace=RootRelative
    boundBone {boneName=dyn_hair_02_l}
    componentAppliedLinearAccClamp=<1000 1000 1000>
    componentLinearAccScale=<0.5 0.5 0.5>
    componentLinearVelScale=<0.3 0.3 0.3>
    physicsBodyDefinitions [{
      boundBone {boneName=dyn_hair_02_l}
      boxExtents=<20 20 20>
      constraintSetup {angularLimitsMax=<0 10 0>, angularLimitsMin=<0 -10 0>, angularTarget=<1 0 0>}
      localJointOffset=<3 0 0>
     }]
    simSpaceSettings {maxAngularAcceleration=45, maxAngularVelocity=8, simSpaceAngularAlpha=0.5}
   }
   n8 = Anim Dynamics - Bone: dyn_hair_03_l!(ComponentPose=@n7, Alpha=1, GravityOverride="0,0,0", ExternalForce="0,0,0") -> (Pose) {
    angularDampingOverride=0.999
    angularSpringConstant=4000
    bAngularSpring=true
    bOverrideAngularDamping=true
    bUseGravityOverride=true
    simulationSpace=RootRelative
    boundBone {boneName=dyn_hair_03_l}
    componentAppliedLinearAccClamp=<1000 1000 1000>
    componentLinearAccScale=<0.5 0.5 0.5>
    componentLinearVelScale=<0.3 0.3 0.3>
    physicsBodyDefinitions [{
      boundBone {boneName=dyn_hair_03_l}
      boxExtents=<20 20 20>
      constraintSetup {angularLimitsMax=<0 5 0>, angularLimitsMin=<0 -5 0>, angularTarget=<1 0 0>}
      localJointOffset=<3 0 0>
     }]
    simSpaceSettings {maxAngularAcceleration=45, maxAngularVelocity=8, simSpaceAngularAlpha=0.25}
   }
   n9 = Anim Dynamics - Bone: dyn_hair_02_r!(ComponentPose=@n8, Alpha=1, GravityOverride="0,0,0", ExternalForce="0,0,0") -> (Pose) {
    angularDampingOverride=0.999
    angularSpringConstant=4000
    bAngularSpring=true
    bOverrideAngularDamping=true
    bUseGravityOverride=true
    simulationSpace=RootRelative
    boundBone {boneName=dyn_hair_02_r}
    componentAppliedLinearAccClamp=<1000 1000 1000>
    componentLinearAccScale=<0.5 0.5 0.5>
    componentLinearVelScale=<0.3 0.3 0.3>
    physicsBodyDefinitions [{
      boundBone {boneName=dyn_hair_02_r}
      boxExtents=<20 20 20>
      constraintSetup {angularLimitsMax=<0 5 0>, angularLimitsMin=<0 0 0>, angularTarget=<1 0 0>}
      localJointOffset=<3 0 0>
     }]
    simSpaceSettings {maxAngularAcceleration=45, maxAngularVelocity=8, simSpaceAngularAlpha=0.25}
   }
   n10 = Anim Dynamics - Bone: dyn_hair_03_r!(ComponentPose=@n9, Alpha=1, GravityOverride="0,0,0", ExternalForce="0,0,0") -> (Pose) {
    angularDampingOverride=0.999
    angularSpringConstant=4000
    bAngularSpring=true
    bOverrideAngularDamping=true
    bUseGravityOverride=true
    simulationSpace=RootRelative
    boundBone {boneName=dyn_hair_03_r}
    componentAppliedLinearAccClamp=<1000 1000 1000>
    componentLinearAccScale=<0.5 0.5 0.5>
    componentLinearVelScale=<0.3 0.3 0.3>
    physicsBodyDefinitions [{
      boundBone {boneName=dyn_hair_03_r}
      boxExtents=<20 20 20>
      constraintSetup {angularLimitsMax=<0 5 0>, angularLimitsMin=<0 0 0>, angularTarget=<1 0 0>}
      localJointOffset=<3 0 0>
     }]
    simSpaceSettings {maxAngularAcceleration=45, maxAngularVelocity=8, simSpaceAngularAlpha=0.25}
   }
   n11 = Blend Poses by bool(
    BlendPose_0=Blend Poses by bool(BlendPose_0=Component To Local(ComponentPose=@n10), BlendPose_1=Component To Local(ComponentPose=@n6)) {blendType=Linear}
    BlendPose_1=Blend Poses by bool(BlendPose_0=Component To Local(ComponentPose=@n4), BlendPose_1=Component To Local(ComponentPose=@n5)) {blendType=Linear}
    bActiveValue=Get bIsInFrontEnd()
   ) {blendType=Linear}
   Output Pose(Result=Component To Local(ComponentPose=Local To Component(LocalPose=@n11))) {layerGroup=None}
   n12 = Layered blend per bone(
    BasePose=Copy Pose From Mesh() {bUseAttachedParent=true}
    BlendPoses_0=Copy Pose From Mesh(SourceMeshComponent=Get Body() -> (Body:Skeletal Mesh Component Object Reference)) {bUseAttachedParent=true}
   ) {layerSetup [{branchFilters [{blendDepth=2, boneName=clavicle_r}, {blendDepth=2, boneName=clavicle_l}, {blendDepth=1, boneName=thigh_l}, {blendDepth=1, boneName=thigh_r}]}]}
   n13 = Control Rig!(Source=@n12, Alpha, BackPack:bool=Get is BackPack on?() -> (is BackPack on?:bool), RBAN:bool=Get RBAN() -> (RBAN:bool)) -> (Pose) {controlRigAssetReference {controlRigAsset=&/F_MED_SprintLuckSlaw_FaceAcc_CtrlRig}}
   Save cached pose 'Main Cache'(Pose=@n13) {cacheName=Main Cache}
  ]}
 EventGraph {nodes [
   n1 = Event OnCharacterCustomizationLoaded()
   > Set Body(Body:Skeletal Mesh Component Object Reference=@n1.CustomBaseMesh) -> (Output_Get:Skeletal Mesh Component Object Reference)
   n2 = CheckForBackbling()
   > Set is BackPack on?(is BackPack on?:bool=OR Boolean!(A=@n2, B=@n2.bIsBackblingACape, C:bool=@n2.bIsBackblingLong) -> (ReturnValue)) -> (Output_Get:bool)
   > n3 = Branch(Condition=Get bIsInFrontEnd())
   > Set Sim Space Settings Output(Sim Space Settings Output:Anim Phys Sim Space Settings Structure=Get Sim Space Settings Front End() -> (Sim Space Settings Front End:Anim Phys Sim Space Settings Structure)) -> (Output_Get:Anim Phys Sim Space Settings Structure)
   Set Sim Space Settings Output(
    execute=@n3.else
    Sim Space Settings Output=Get Sim Space Settings Level() -> (Sim Space Settings Level:Anim Phys Sim Space Settings Structure)
   )
   Event BlueprintUpdateAnimation()
   > n4 = "Gravity Override State Function (Parameter Asset) (Nativized)"(
    CharacterPartAnimInstance=Self-Reference()
    GameTimeInSeconds=GetGameTimeInSeconds()
    Parameters=/BRCosmetics/Characters/Player/Female/Medium/Bodies/F_MED_SprintLuckSlaw/Meshes/F_MED_SprintLuckSlaw_Parameters
   )
   > Set GravityOverride(GravityOverride=@n4)
   > Set RBAN(RBAN:bool=NOT Boolean(A=Get bRigidBodyEnabled())) -> (Output_Get:bool)
   > Set RBAN Settings Out(RBAN Settings Out=@n4.OutRigidBodyAnimNodeParameters)
  ]}
]
variables [
 Body {type=SkeletalMeshComponent}
 GravityOverride {type=Vector}
 is BackPack on? {type=boolean}
 RBAN {type=boolean}
 RBAN Settings Out {type=FortRigidBodyAnimNodeParameters}
 Sim Space Settings Front End {
  type=AnimPhysSimSpaceSettings
  default {maxAngularAcceleration=10000, maxAngularVelocity=10000, simSpaceAngularAlpha=0, externalAngularVelocity=<0 0 0>}
 }
 Sim Space Settings Level {
  type=AnimPhysSimSpaceSettings
  default {maxAngularAcceleration=10000, maxAngularVelocity=10000, simSpaceAngularAlpha=0, externalAngularVelocity=<0 0 0>}
 }
 Sim Space Settings Output {
  type=AnimPhysSimSpaceSettings
  default {maxAngularAcceleration=56, maxAngularVelocity=8, simSpaceAngularAlpha=0.75, externalAngularVelocity=<0 0 0>}
 }
 Targeting {type=boolean}
]
